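# BP_STPS_3DSelect is an Editor Utility Blueprint that holds settings for a 3D Select editor tool.
# It has no graph logic, only variables for class and label include/exclude filtering, object visibility, bounding box display, gizmo sensitivity and edge scaling.
# By default it filters in Exclude mode with "landscape" as an excluded label and shows all objects.
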EditorUtilityBlueprint /Game/Blueprints/Blutilities/ToolsWrkshp/ScrTools/ST_3DSelect/BP_STPS_3DSelect
graphs [EventGraph {links [], nodes []}]
variables [
 Bounding Box Color {type=Color, default=<rgba 0.9216 0 0 1>}
 Bounding Box Line Thickness {type=number, default=3}
 Exclude Class {type="array<string>", default []}
 Exclude Label {type="array<string>", default [landscape]}
 Filter Mode {type=E_3DSelect_FilterOptions, default=Exclude}
 Include Class {type="array<string>", default []}
 Include Label {type="array<string>", default []}
 Object Visibility {type=E_3DSelect_HidingOptions, default=Show All}
 Show Bounding Box Mesh {type=boolean}
 Transform Gizmo Sensitivity {type=number, default=1}
 Use Edge Scaling {type=boolean, default=true}
]
