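# Editor scriptable tool that selects level actors using an adjustable 3D bounding box, with readout and shortcut panels.
# Per-face translation gizmos or one combined gizmo, plus the mouse wheel, resize the box and refresh the selection at throttled intervals.
# Include or exclude filters match actor class names and labels by substring.
# Tick polling detects selection changes, non-selected actors can be temporarily hidden, and a transient cube mesh visualizes the volume.
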EditorUtilityBlueprint /Game/Blueprints/Blutilities/ToolsWrkshp/ScrTools/ST_3DSelect/BP_ST_3DSelect
graphs [
 ApplyFilters {nodes [
   n1 = ApplyFilters!() -> (then, UnfilteredActors:Array of Actor Object References) {is=title K2Node_FunctionEntry}
   > Set Unfiltered Actors(Unfiltered Actors:Array of Actor Object References=@n1) -> (Output_Get:Array of Actor Object References)
   > n2 = Sequence!(execute) -> (then_0, then_1, then_2:Exec, then_3:Exec)
   > UpdateFilterVariables()
   n3 = Switch on E_3DSelect_FilterOptions!(
    execute=@n2.then_3
    Selection:E_3DSelect_FilterOptions Enum=Get Filter Mode!(self:BP STPS 3DSelect Object Reference=Get Settings() -> (Settings:BP STPS 3DSelect Object Reference)) -> (Filter Mode:E_3DSelect_FilterOptions Enum=NewEnumerator0)
   ) -> (NewEnumerator0:Exec, NewEnumerator1:Exec, NewEnumerator2:Exec) {is=title K2Node_SwitchEnum}
   n4 = For Each Loop(
    Exec=@n3.NewEnumerator1
    Array:Array of Actor Object References=Get Unfiltered Actors() -> (Unfiltered Actors:Array of Actor Object References)
   ) -> (Array Element:Actor Object Reference)
   Return Node!(
    execute=@n4.Completed
    FilteredActorsKept:Array of Actor Object References=Get Filtered Actors() -> (Filtered Actors:Array of Actor Object References)
    FilteredActorsRemoved:Array of Actor Object References=Get Excluded Actors() -> (Excluded Actors:Array of Actor Object References)
   )
   > Set IncludedByClass(IncludedByClass:bool=false) -> (Output_Get:bool)
   > Set IncludedByLabel(IncludedByLabel:bool=false) -> (Output_Get:bool)
   > Set CurActorBeingFiltered(CurActorBeingFiltered:Actor Object Reference=@n4) -> (Output_Get:Actor Object Reference)
   > Set CurrentActorBeingFilteredClass(CurrentActorBeingFilteredClass:String=Get Class Display Name(Class=Get Class(Object=@n4))) -> (Output_Get:String)
   > n5 = For Each Loop with Break!(Exec:Exec, Array:Array of Strings=Get CurrentClassesToInclude() -> (CurrentClassesToInclude:Array of Strings), Break:Exec=@n16.then) -> (LoopBody:Exec, Array Element:String, Array Index:int, Completed:Exec)
   n6 = Branch(execute=@n5.Completed, Condition=Get IncludedByClass() -> (IncludedByClass:bool))
   > Add(
    TargetArray:Array of Actor Object References=Get Filtered Actors()
    NewItem:"Actor Object Reference (by ref)"=Get CurActorBeingFiltered() -> (CurActorBeingFiltered:Actor Object Reference)
   ) {is=title K2Node_CallArrayFunction}
   Set CurActorBeingFilteredLabel(execute=@n6.else, CurActorBeingFilteredLabel:String=GetActorLabel(self=Get CurActorBeingFiltered(), bCreateIfNone=false)) -> (Output_Get:String)
   > n7 = For Each Loop with Break(Array=Get CurrentLabelsToInclude() -> (CurrentLabelsToInclude:Array of Strings), Break=@n13.then)
   n8 = Get IncludedByClass()
   n9 = Get IncludedByLabel() -> (IncludedByLabel:bool)
   Branch(execute=@n7.Completed, Condition=AND Boolean(A=NOT Boolean(A=@n8), B=NOT Boolean(A=@n9)))
   > Add(TargetArray=Get Excluded Actors(), NewItem=Get CurActorBeingFiltered())
   n10 = Branch(execute=@n7.LoopBody, Condition=Get ExcludedByLabel() -> (ExcludedByLabel:bool))
   n11 = Branch(execute=@n10.else, Condition=IsEmpty(InString=@n7))
   n12 = Get CurActorBeingFilteredLabel() -> (CurActorBeingFilteredLabel:String)
   Branch(execute=@n11.else, Condition="integer >= integer"(A=FindSubstring(SearchIn=@n12, Substring=@n7)))
   > Add(TargetArray=Get Filtered Actors(), NewItem=Get CurActorBeingFiltered())
   > n13 = Set IncludedByLabel(IncludedByLabel=true)
   n14 = Branch(execute=@n5.LoopBody, Condition=IsEmpty(InString=@n5))
   n15 = Get CurrentActorBeingFilteredClass() -> (CurrentActorBeingFilteredClass:String)
   Branch(execute=@n14.else, Condition="integer >= integer"(A=FindSubstring(SearchIn=@n15, Substring=@n5)))
   > n16 = Set IncludedByClass(IncludedByClass=true)
   n17 = For Each Loop(Exec=@n3.NewEnumerator2, Array=Get Unfiltered Actors())
   Return Node(execute=@n17.Completed, FilteredActorsKept=Get Filtered Actors(), FilteredActorsRemoved=Get Excluded Actors())
   > Set ExcludedByClass(ExcludedByClass:bool=false) -> (Output_Get:bool)
   > Set ExcludedByLabel(ExcludedByLabel:bool=false) -> (Output_Get:bool)
   > Set CurActorBeingFiltered(CurActorBeingFiltered=@n17)
   > Set CurrentActorBeingFilteredClass(CurrentActorBeingFilteredClass=Get Class Display Name(Class=Get Class(Object=@n17)))
   > n18 = For Each Loop with Break(Array=Get CurrentClassesToExclude() -> (CurrentClassesToExclude:Array of Strings), Break=@n28.then)
   n19 = Branch(execute=@n18.Completed, Condition=Get ExcludedByClass() -> (ExcludedByClass:bool))
   > Add(TargetArray=Get Excluded Actors(), NewItem=Get CurActorBeingFiltered())
   Branch(execute=@n19.else, Condition=NOT Boolean(A=Get ExcludedByClass()))
   > Set CurActorBeingFilteredLabel(CurActorBeingFilteredLabel=GetActorLabel(self=Get CurActorBeingFiltered(), bCreateIfNone=false))
   > n20 = For Each Loop(Array:Array of Strings=Get CurrentLabelsToExclude() -> (CurrentLabelsToExclude:Array of Strings)) -> (Array Element:String)
   n21 = Get ExcludedByClass()
   n22 = Get ExcludedByLabel()
   Branch(execute=@n20.Completed, Condition=AND Boolean(A=NOT Boolean(A=@n21), B=NOT Boolean(A=@n22)))
   > Add(TargetArray=Get Filtered Actors(), NewItem=Get CurActorBeingFiltered())
   n23 = Branch(execute=@n20.LoopBody, Condition=Get ExcludedByLabel())
   n24 = Branch(execute=@n23.else, Condition=IsEmpty(InString=@n20))
   n25 = Get CurActorBeingFilteredLabel()
   Branch(execute=@n24.else, Condition="integer >= integer"(A=FindSubstring(SearchIn=@n25, Substring=@n20)))
   > Add(TargetArray=Get Excluded Actors(), NewItem=Get CurActorBeingFiltered())
   > Set ExcludedByLabel(ExcludedByLabel=true)
   n26 = Branch(execute=@n18.LoopBody, Condition=IsEmpty(InString=@n18))
   n27 = Get CurrentActorBeingFilteredClass()
   Branch(execute=@n26.else, Condition="integer >= integer"(A=FindSubstring(SearchIn=@n27, Substring=@n18)))
   > n28 = Set ExcludedByClass(ExcludedByClass=true)
   Return Node(
    execute=@n3.NewEnumerator0
    FilteredActorsKept=Get UnfilteredActors() -> (UnfilteredActors:Array of Actor Object References)
    FilteredActorsRemoved=Get Excluded Actors()
   )
  ]}
 CanBeginClickDragSequence {nodes [
   n1 = CanBeginClickDragSequence!() -> (
    then
    PressPos_WorldRay:Ray Structure
    PressPos_bHas2D:bool=~
    PressPos_ScreenPosition:Vector 2D Structure
    Modifiers_bShiftDown:bool=~
    Modifiers_bCtrlDown:bool=~
    Modifiers_bAltDown:bool=~
    MouseButton:EScriptableToolMouseButton Enum=LeftButton
   ) {is=title K2Node_FunctionEntry}
   > Return Node!(
    execute
    ReturnValue_bHit:bool=False
    ReturnValue_HitDepth:double=340282346638528859811704183484516925440
    ReturnValue_HitNormal:Vector="0,0,1"
    ReturnValue_bHasHitNormal:bool=False
    ReturnValue_HitIdentifier:int=0
    ReturnValue_HitObject:Object Reference
   )
   AND Boolean(A=@n1.Modifiers_bShiftDown, B=@n1.Modifiers_bCtrlDown)
  ]}
 CaptureCheckFilterClassKey {nodes [
   n1 = CaptureCheckFilterClassKey!() -> (
    then
    InputDeviceState_bShiftKeyDown:bool=~
    InputDeviceState_bAltKeyDown:bool=~
    InputDeviceState_bCtrlKeyDown:bool=~
    InputDeviceState_bCmdKeyDown:bool=~
    InputDeviceState_Keyboard:Keyboard Input Device State Structure
    InputDeviceState_Mouse:Mouse Input Device State Structure
   ) {is=title K2Node_FunctionEntry}
   > Return Node!(execute, ReturnValue:bool=NOT Boolean(A=@n1))
  ]}
 CaptureCheckFilterStringKey {nodes [
   CaptureCheckFilterStringKey!() -> (
    then
    InputDeviceState_bShiftKeyDown:bool=~
    InputDeviceState_bAltKeyDown:bool=~
    InputDeviceState_bCtrlKeyDown:bool=~
    InputDeviceState_bCmdKeyDown:bool=~
    InputDeviceState_Keyboard:Keyboard Input Device State Structure
    InputDeviceState_Mouse:Mouse Input Device State Structure
   ) {is=title K2Node_FunctionEntry}
   > Return Node(ReturnValue=true)
  ]}
 CaptureMouseWheel {nodes [
   CaptureMouseWheel!() -> (then, InputDeviceState:"Input Device State Structure (by ref)") {is=title K2Node_FunctionEntry}
   > Return Node(ReturnValue=Get IsModKeyPressed() -> (IsModKeyPressed:bool))
  ]}
 ControlFadeCube {nodes [
   ControlFadeCube() {is=title K2Node_FunctionEntry}
   > n1 = Branch(Condition=Get Show Bounding Box Mesh!(self:BP STPS 3DSelect Object Reference=Get Settings()) -> (Show Bounding Box Mesh:bool))
   > n2 = Is Valid(InputObject=Get FadeCubeActor() -> (FadeCubeActor:Static Mesh Actor Object Reference))
   n3 = SpawnActorFromClass!(
    execute=@n2.Is Not Valid
    ActorClass:Actor Class Reference=&StaticMeshActor
    Location:Vector=Get SelectionBoundOrigin() -> (SelectionBoundOrigin:Vector)
    Rotation:Rotator
    bTransient:bool=true
   ) -> (then, ReturnValue:Static Mesh Actor Object Reference)
   > n4 = Set FadeCubeActor(FadeCubeActor:Static Mesh Actor Object Reference=@n3) -> (Output_Get:Static Mesh Actor Object Reference)
   > SetEditorProperty(Object=@n4, PropertyName=bIsEditorOnlyActor, PropertyValue=MakeLiteralBool(Value=true))
   > SetActorLabel(self=@n4, NewActorLabel=SelectionSizerActor, bMarkDirty=true)
   > n5 = Set FadeCubeActorStaticMeshComponent(FadeCubeActorStaticMeshComponent:Static Mesh Component Object Reference=Get StaticMeshComponent(self=@n4)) -> (Output_Get:Static Mesh Component Object Reference)
   > SetStaticMesh(self=@n5, NewMesh=/Game/Blueprints/Blutilities/ToolsWrkshp/ScrTools/ST_3DSelect/SupportFiles/SM_3DSelect_VolCube)
   > SetEditorProperty(
    Object=Get FadeCubeActorStaticMeshComponent() -> (FadeCubeActorStaticMeshComponent:Static Mesh Component Object Reference)
    PropertyName=bIsEditorOnly
    PropertyValue=MakeLiteralBool(Value=true)
   )
   > SetMaterial(
    self=Get FadeCubeActorStaticMeshComponent()
    ElementIndex=0
    Material=/Game/Blueprints/Blutilities/ToolsWrkshp/ScrTools/ST_3DSelect/SupportFiles/M_3DSelect_Volume
   )
   > SetVisibility!(
    execute
    self:Scene Component Object Reference=Get FadeCubeActorStaticMeshComponent()
    bNewVisibility:bool=true
    bPropagateToChildren:bool=false
   ) -> (then)
   n6 = Get FadeCubeActor()
   Is Valid(exec=@n1.else, InputObject=@n6)
   > Destroy Actor(self=@n6)
  ]}
 DetectSelectionChanged {nodes [
   DetectSelectionChanged() {is=title K2Node_FunctionEntry}
   > n1 = Branch(Condition=Get SuspendSelChangeDetection() -> (SuspendSelChangeDetection:bool))
   n2 = WaitForTime!(execute=@n1.else, self, PauseTime:double=0.5) -> (then, Active:bool)
   > Branch(Condition=@n2)
   > n3 = GetSelectedLevelActors(self=Get EditorActorSubsystem())
   > Set CurSelSet(CurSelSet:Array of Actor Object References=@n3) -> (Output_Get:Array of Actor Object References)
   > Set SetsTheSame(SetsTheSame:bool=Identical!(
     ArrayA:Array of Actor Object References=Get CurSelSet() -> (CurSelSet:Array of Actor Object References)
     ArrayB:Array of Actor Object References=Get DetectSelChangedLastSet() -> (DetectSelChangedLastSet:Array of Actor Object References)
    ) -> (ReturnValue:bool) {is=title K2Node_CallArrayFunction}) -> (Output_Get:bool)
   > Set DetectSelChangedLastSet(DetectSelChangedLastSet:Array of Actor Object References=Get CurSelSet()) -> (Output_Get:Array of Actor Object References)
   > n4 = Branch(Condition=Get SetsTheSame() -> (SetsTheSame:bool))
   OnSelectionChanged!(execute=@n4.else, self, NewSelection:Array of Actor Object References=Get CurSelSet()) -> (then)
   > OnObjectVisChanged!(execute, self, PropertySet:Scriptable Interactive Tool Property Set Object Reference, PropertyName:String) -> (then)
  ]}
 DoScriptShutdown {nodes [
   DoScriptShutdown() {is=title K2Node_FunctionEntry}
   > n1 = Sequence()
   > ShowAll()
   > Set Object Visibility!(execute, Object Visibility:E_3DSelect_HidingOptions Enum=NewEnumerator0, self:BP STPS 3DSelect Object Reference=Get Settings()) -> (then, Output_Get:E_3DSelect_HidingOptions Enum=NewEnumerator0)
   SavePropertySetSettings(execute=@n1.then_1, PropertySet=Get Settings(), SaveKey=Get PropSetSaveKey() -> (PropSetSaveKey:String))
   n2 = Get Readout Box() -> (Readout Box:BP UW Expandable Box Object Reference)
   Set ReadoutOpenOnExit!(
    execute=@n1.then_2
    ReadoutOpenOnExit:bool=IsChecked(self=Get ExpandButton!(self:BP UW Expandable Box Object Reference=@n2) -> (ExpandButton:Check Box Object Reference))
    self:BP STPS Shared Widget State Object Reference=Get SharedState Settings() -> (SharedState Settings:BP STPS Shared Widget State Object Reference)
   ) -> (then, Output_Get:bool)
   n3 = Get Shortcut Box() -> (Shortcut Box:BP UW Expandable Box Object Reference)
   > Set ShortcutsOpenOnExit!(
    execute
    ShortcutsOpenOnExit:bool=IsChecked(self=Get ExpandButton(self=@n3))
    self:BP STPS Shared Widget State Object Reference=Get SharedState Settings()
   ) -> (then, Output_Get:bool)
   > SavePropertySetSettings(PropertySet=Get SharedState Settings(), SaveKey=Get PropSetSaveKey())
   HandleScaleGizmos!(execute=@n1.then_3, self, IsCreate:bool=false) -> (then)
   > Destroy Actor(self=Get FadeCubeActor())
  ]}
 EventGraph {nodes [
   Event OnScriptSetup()
   > Setup()
   n1 = Event OnGizmoTransformChanged!() -> (OutputDelegate, then, GizmoIdentifier:String, NewTransform_Location:Vector, NewTransform_Rotation:Rotator, NewTransform_Scale:Vector) {is=title K2Node_Event}
   > GizmoTransformChanged!(
    execute
    self
    Gizmo Identifier:String=@n1.GizmoIdentifier
    NewTransformLocation:Vector=@n1.NewTransform_Location
    NewTransformRotation:Rotator=@n1.NewTransform_Rotation
    NewTransformScale:Vector=@n1.NewTransform_Scale
   ) -> (then)
   n2 = Event OnScriptTick()
   > Set CurTime(CurTime=float + float(A=@n2.DeltaTime, B=Get CurTime()))
   > DetectSelectionChanged() {is=title K2Node_CallFunction}
   > n3 = Sequence!(execute) -> (then_0, then_1)
   > n4 = Branch(Condition=Get FirstInitRun())
   Set FirstInitRun(execute=@n4.else)
   > n5 = Sequence!(execute) -> (then_0, then_1, then_2:Exec)
   n6 = Get Header Widget()
   > n7 = AddExpandableBox(self=@n6)
   > n8 = Set Readout Box(Readout Box:BP UW Expandable Box Object Reference=@n7) -> (Output_Get:BP UW Expandable Box Object Reference)
   > SetUpBox(
    self=@n8
    Name=Get Readout Box Label!(self:BP UW Shared Header Widget Object Reference=@n6) -> (Readout Box Label:Text)
    OpenByDefault=Get ReadoutOpenOnExit!(self:BP STPS Shared Widget State Object Reference=Get SharedState Settings()) -> (ReadoutOpenOnExit:bool)
   )
   > n9 = Sequence!(execute) -> (then_0, then_1, then_2, then_3:Exec, then_4:Exec)
   n10 = Set Readout Line Selected(execute=@n11.then, Readout Line Selected:BP UW Readout Line Object Reference=@n11) -> (Output_Get:BP UW Readout Line Object Reference)
   n11 = SetupLine!(
    execute=@n9.then_0
    self:BP UW Expandable Box Object Reference=Get Readout Box()
    CurrentLine:BP UW Readout Line Object Reference=@n10
    DynamicLine:Vertical Box Object Reference
    Label:Text=Get Selected Line Label() -> (Selected Line Label:Text)
    Value:Text=Get Selected Line Value() -> (Selected Line Value:Text)
   ) -> (then, Readout Line:BP UW Readout Line Object Reference)
   n12 = Set Readout Line Actor(execute=@n13.then, Readout Line Actor:BP UW Readout Line Object Reference=@n13) -> (Output_Get:BP UW Readout Line Object Reference)
   n13 = SetupLine(
    execute=@n9.then_1
    self=Get Readout Box()
    CurrentLine=@n12
    Label=Get Actor Line Label() -> (Actor Line Label:Text)
    Value=Get Actor Line Value() -> (Actor Line Value:Text)
   )
   n14 = Set Readout Line Class(execute=@n15.then, Readout Line Class:BP UW Readout Line Object Reference=@n15) -> (Output_Get:BP UW Readout Line Object Reference)
   n15 = SetupLine(
    execute=@n9.then_2
    self=Get Readout Box()
    CurrentLine=@n14
    Label=Get ClassLineLabel() -> (ClassLineLabel:Text)
    Value=Get ClassLineValue() -> (ClassLineValue:Text)
   )
   n16 = Set Readout Line Dimensions(execute=@n17.then, Readout Line Dimensions:BP UW Readout Line Object Reference=@n17) -> (Output_Get:BP UW Readout Line Object Reference)
   n17 = SetupLine(
    execute=@n9.then_3
    self=Get Readout Box()
    CurrentLine=@n16
    Label=Get Box Dimensions Label() -> (Box Dimensions Label:Text)
    Value=Get Box Dimensions Line Empty Value() -> (Box Dimensions Line Empty Value:Text)
   )
   n18 = Set Readout Line Nearest Snap Grid(execute=@n19.then, Readout Line Nearest Snap Grid:BP UW Readout Line Object Reference=@n19) -> (Output_Get:BP UW Readout Line Object Reference)
   n19 = SetupLine(
    execute=@n9.then_4
    self=Get Readout Box()
    CurrentLine=@n18
    Label=Get Nearest Grid Label() -> (Nearest Grid Label:Text)
    Value=Get Nearest Grid Value() -> (Nearest Grid Value:Text)
   )
   n20 = Get Header Widget()
   n21 = AddExpandableBox(execute=@n5.then_1, self=@n20)
   > n22 = Set Shortcut Box(Shortcut Box:BP UW Expandable Box Object Reference=@n21) -> (Output_Get:BP UW Expandable Box Object Reference)
   > n23 = SetUpBox(
    self=@n22
    Name=Get Shortcut Box Label!(self:BP UW Shared Header Widget Object Reference=@n20) -> (Shortcut Box Label:Text)
    OpenByDefault=Get ShortcutsOpenOnExit!(self:BP STPS Shared Widget State Object Reference=Get SharedState Settings()) -> (ShortcutsOpenOnExit:bool)
   )
   n24 = Set ShortcutLines(execute=@n25.then, ShortcutLines:Array of BP UW Readout Line Object References=@n25) -> (Output_Get:Array of BP UW Readout Line Object References)
   n25 = AddShortcuts!(
    execute=@n23.then
    self:BP UW Expandable Box Object Reference=Get Shortcut Box()
    ShortcutMap:Map of Strings to Strings=Get ShortcutLabels() -> (ShortcutLabels:Map of Strings to Strings)
    ExistingShortcutLines:Array of BP UW Readout Line Object References=@n24
    ShortcutMapKeys:Array of Strings=Get ShortcutKeys() -> (ShortcutKeys:Array of Strings)
   ) -> (then, AllShortcutLines:Array of BP UW Readout Line Object References)
   UpdateSelectedActorText(execute=@n5.then_2)
   > Set Last Clear Time(Last Clear Time:double=0) -> (Output_Get:double)
   n26 = TimedClearUserWarning!(
    execute=@n3.then_1
    ClearAfterTime:double=Get Clear Warning After Time() -> (Clear Warning After Time:double)
    CurTime:double=Get CurTime()
    LastClearTime:double=Get Last Clear Time() -> (Last Clear Time:double)
    Scriptable Tool Obj Ref:Scriptable Interactive Tool Object Reference=Self-Reference()
   ) -> (then, NewLastClearTime:double)
   > Set Last Clear Time(Last Clear Time=@n26)
   n27 = Event OnGizmoTransformStateChange!() -> (OutputDelegate, then, GizmoIdentifier:String, CurrentTransform:Transform, ChangeType:EScriptableToolGizmoStateChangeType Enum=BeginTransform) {is=title K2Node_Event}
   > Switch on EScriptableToolGizmoStateChangeType!(execute, Selection:EScriptableToolGizmoStateChangeType Enum=@n27.ChangeType) -> (BeginTransform:Exec, EndTransform:Exec, UndoRedo:Exec) {is=title K2Node_SwitchEnum}
   > Set ExtentBeforeGizmo(ExtentBeforeGizmo:Vector=Get SelectionBoundExtent() -> (SelectionBoundExtent:Vector)) -> (Output_Get:Vector)
   > Set OriginBeforeGizmo(OriginBeforeGizmo:Vector=Get SelectionBoundOrigin()) -> (Output_Get:Vector)
   Event OnScriptShutdown()
   > DoScriptShutdown() {is=title K2Node_CallFunction}
  ]}
 ForceSnapBoundsToSelection {nodes [
   ForceSnapBoundsToSelection() {is=title K2Node_FunctionEntry}
   > UpdateSelectionBounds!(execute, self, ForceUpdate:bool=true) -> (then)
   > UpdateSelectedActorText()
   > UpdateBBox!(execute, self, BoundOrigin:Vector=Get SelectionBoundOrigin(), BoundExtent:Vector=Get SelectionBoundExtent(), Update Properties:bool=false) -> (then)
  ]}
 GetBoxFaceLocation {nodes [
   n1 = GetBoxFaceLocation!() -> (then, FaceIndex:int) {is=title K2Node_FunctionEntry}
   n2 = Get SelectionBoundOrigin()
   n3 = Get SelectionBoundExtent()
   > Return Node!(
    execute
    Center Point:Vector=GetBoxFaceCenter!(Box:Box Structure=MakeBoxFromCenterExtents!(Center:Vector=@n2, Extents:Vector=@n3) -> (ReturnValue:Box Structure), FaceIndex:int=@n1) -> (FaceNormal:Vector, ReturnValue:Vector).ReturnValue
   )
  ]}
 GetSelectionBounds {nodes [
   GetSelectionBounds() {is=title K2Node_FunctionEntry}
   > n1 = GetSelectedLevelActors(self=Get EditorActorSubsystem())
   > n2 = GetActorArrayBounds!(
    execute
    Actors:Array of Actor Object References=@n1
    bOnlyCollidingComponents:bool=Get Only Colliding Components() -> (Only Colliding Components:bool)
   ) -> (then, Center:Vector, BoxExtent:Vector)
   > Set SelectionBoundOrigin(SelectionBoundOrigin:Vector=@n2) -> (Output_Get:Vector)
   > Set SelectionBoundExtent(SelectionBoundExtent:Vector=@n2.BoxExtent) -> (Output_Get:Vector)
   > Return Node!(execute)
  ]}
 GizmoTransformChanged {nodes [
   GizmoTransformChanged!() -> (then, Gizmo Identifier:String, NewTransformLocation:Vector=~, NewTransformRotation:Rotator=~, NewTransformScale:Vector=~) {is=title K2Node_FunctionEntry}
   > n1 = Sequence!(execute) -> (then_0, then_1)
   > n2 = Branch(Condition=Get Use Edge Scaling!(self:BP STPS 3DSelect Object Reference=Get Settings()) -> (Use Edge Scaling:bool))
   > Sequence()
   > n3 = Switch on String!(execute:Exec, Selection:String=Get Gizmo Identifier() -> (Gizmo Identifier:String)) -> (Default:Exec, BoundEdge_0:Exec, BoundEdge_1:Exec, BoundEdge_2:Exec, BoundEdge_3:Exec, BoundEdge_4:Exec, BoundEdge_5:Exec) {pinNames [BoundEdge_0,BoundEdge_1,BoundEdge_2,BoundEdge_3,BoundEdge_4,BoundEdge_5]}
   n4 = Get ExtentBeforeGizmo() -> (ExtentBeforeGizmo:Vector)
   n5 = Get OriginBeforeGizmo() -> (OriginBeforeGizmo:Vector)
   n6 = float - float(
    A=BreakVector(InVec=Get NewTransformLocation() -> (NewTransformLocation:Vector))
    B=float + float(A=BreakVector(InVec=@n5), B=BreakVector(InVec=@n4))
   )
   n7 = SafeDivide(A=@n6, B=2)
   Set LocalNewExtent(execute=@n3.BoundEdge_0, LocalNewExtent:Vector=vector + vector(A=Get ExtentBeforeGizmo(), B=MakeVector(X=@n7, Y=0))) -> (Output_Get:Vector)
   n8 = SafeDivide(A=@n6, B=2)
   > Set LocalNewLocation(LocalNewLocation:Vector=vector + vector(A=Get OriginBeforeGizmo(), B=MakeVector(X=@n8, Y=0))) -> (Output_Get:Vector)
   > Set SelectionBoundOrigin(SelectionBoundOrigin=Get LocalNewLocation() -> (LocalNewLocation:Vector))
   > Set SelectionBoundExtent(SelectionBoundExtent=Get LocalNewExtent() -> (LocalNewExtent:Vector))
   n9 = Get ExtentBeforeGizmo()
   n10 = Get OriginBeforeGizmo()
   n11 = float - float(A=float - float(A=BreakVector(InVec=@n10), B=BreakVector(InVec=@n9)), B=BreakVector(InVec=Get NewTransformLocation()))
   n12 = SafeDivide(A=@n11, B=2)
   Set LocalNewExtent(execute=@n3.BoundEdge_1, LocalNewExtent=vector + vector(A=Get ExtentBeforeGizmo(), B=MakeVector(X=@n12, Y=0)))
   n13 = SafeDivide(A=@n11, B=2)
   > Set LocalNewLocation(LocalNewLocation=vector - vector(A=Get OriginBeforeGizmo(), B=MakeVector(X=@n13, Y=0)))
   > Set SelectionBoundOrigin(SelectionBoundOrigin=Get LocalNewLocation())
   > Set SelectionBoundExtent(SelectionBoundExtent=Get LocalNewExtent())
   n14 = Get ExtentBeforeGizmo()
   n15 = Get OriginBeforeGizmo()
   n16 = float - float(A=BreakVector(InVec=Get NewTransformLocation()).Y, B=float + float(A=BreakVector(InVec=@n15).Y, B=BreakVector(InVec=@n14).Y))
   n17 = SafeDivide(A=@n16, B=2)
   Set LocalNewExtent(execute=@n3.BoundEdge_2, LocalNewExtent=vector + vector(A=Get ExtentBeforeGizmo(), B=MakeVector(Y=@n17)))
   n18 = SafeDivide(A=@n16, B=2)
   > Set LocalNewLocation(LocalNewLocation=vector + vector(A=Get OriginBeforeGizmo(), B=MakeVector(Y=@n18)))
   > Set SelectionBoundOrigin(SelectionBoundOrigin=Get LocalNewLocation())
   > Set SelectionBoundExtent(SelectionBoundExtent=Get LocalNewExtent())
   n19 = Get ExtentBeforeGizmo()
   n20 = Get OriginBeforeGizmo()
   n21 = float - float(A=float - float(A=BreakVector(InVec=@n20).Y, B=BreakVector(InVec=@n19).Y), B=BreakVector(InVec=Get NewTransformLocation()).Y)
   n22 = SafeDivide(A=@n21, B=2)
   Set LocalNewExtent(execute=@n3.BoundEdge_3, LocalNewExtent=vector + vector(A=Get ExtentBeforeGizmo(), B=MakeVector(Y=@n22)))
   n23 = SafeDivide(A=@n21, B=2)
   > Set LocalNewLocation(LocalNewLocation=vector - vector(A=Get OriginBeforeGizmo(), B=MakeVector(Y=@n23)))
   > Set SelectionBoundOrigin(SelectionBoundOrigin=Get LocalNewLocation())
   > Set SelectionBoundExtent(SelectionBoundExtent=Get LocalNewExtent())
   n24 = Get ExtentBeforeGizmo()
   n25 = Get OriginBeforeGizmo()
   n26 = float - float(A=BreakVector(InVec=Get NewTransformLocation()).Z, B=float + float(A=BreakVector(InVec=@n25).Z, B=BreakVector(InVec=@n24).Z))
   n27 = SafeDivide(A=@n26, B=2)
   Set LocalNewExtent(execute=@n3.BoundEdge_4, LocalNewExtent=vector + vector(A=Get ExtentBeforeGizmo(), B=MakeVector(Y=0, Z=@n27)))
   n28 = SafeDivide(A=@n26, B=2)
   > Set LocalNewLocation(LocalNewLocation=vector + vector(A=Get OriginBeforeGizmo(), B=MakeVector(Y=0, Z=@n28)))
   > Set SelectionBoundOrigin(SelectionBoundOrigin=Get LocalNewLocation())
   > Set SelectionBoundExtent(SelectionBoundExtent=Get LocalNewExtent())
   n29 = Get ExtentBeforeGizmo()
   n30 = Get OriginBeforeGizmo()
   n31 = float - float(A=float - float(A=BreakVector(InVec=@n30).Z, B=BreakVector(InVec=@n29).Z), B=BreakVector(InVec=Get NewTransformLocation()).Z)
   n32 = SafeDivide(A=@n31, B=2)
   Set LocalNewExtent(execute=@n3.BoundEdge_5, LocalNewExtent=vector + vector(A=Get ExtentBeforeGizmo(), B=MakeVector(Y=0, Z=@n32)))
   n33 = SafeDivide(A=@n31, B=2)
   > Set LocalNewLocation(LocalNewLocation=vector - vector(A=Get OriginBeforeGizmo(), B=MakeVector(Y=0, Z=@n33)))
   > Set SelectionBoundOrigin(SelectionBoundOrigin=Get LocalNewLocation())
   > Set SelectionBoundExtent(SelectionBoundExtent=Get LocalNewExtent())
   n34 = Set SelectionBoundOrigin(execute=@n2.else, SelectionBoundOrigin=Get NewTransformLocation())
   n35 = float * float(A=Get Transform Gizmo Sensitivity!(self:BP STPS 3DSelect Object Reference=Get Settings()) -> (Transform Gizmo Sensitivity:double), B=2) {is=title K2Node_PromotableOperator}
   n36 = vector - vector(A=Get NewTransformScale() -> (NewTransformScale:Vector), B=Get ExtentBeforeGizmo())
   n37 = BreakVector(InVec=vector + vector(A=Get ExtentBeforeGizmo(), B=vector * vector(A=@n36, B=@n35) -> (ReturnValue=~)))
   n38 = MakeVector(
    X="Clamp (Float)"(Value=@n37, Min=0.1, Max=1000000000)
    Y="Clamp (Float)"(Value=@n37.Y, Min=0.1, Max=1000000000)
    Z="Clamp (Float)"(Value=@n37.Z, Min=0.1, Max=1000000000)
   )
   > Set SelectionBoundExtent(SelectionBoundExtent=@n38)
   > SetGizmoTransform!(
    execute
    self:Scriptable Interactive Tool Object Reference
    Identifier:String=Get Gizmo Identifier()
    NewTransform:Transform=MakeTransform(Location=@n34, Rotation="0, 0, 0", Scale=@n38)
    bUndoable:bool=false
   ) -> (then)
   UpdateBBox(execute=@n1.then_1, BoundOrigin=Get SelectionBoundOrigin(), BoundExtent=Get SelectionBoundExtent(), Update Properties=true)
   > n39 = WaitForTime(PauseTime=0.05)
   > Branch(Condition=@n39)
   > UpdateSelectedActors!(execute, self, BoxPos:Vector=Get SelectionBoundOrigin(), BoxExtent:Vector=Get SelectionBoundExtent()) -> (then)
   > OnObjectVisChanged()
   n40 = For Each Loop(Array=Get AllGizmoIdentifiers() -> (AllGizmoIdentifiers:Array of Strings))
   > n41 = Branch(Condition="Equal Exactly (String)"(A=Get Gizmo Identifier(), B=@n40))
   n42 = GetBoxFaceLocation!(
    execute=@n41.else
    self
    FaceIndex:int="Get (a ref)"(Array=Get IdMapToBoxFace() -> (IdMapToBoxFace:Array of Integers), Dimension 1=@n40.Array Index)
   ) -> (then, Center Point:Vector) {is=title K2Node_CallFunction}
   > n43 = GetGizmoTransform!(execute, self:Scriptable Interactive Tool Object Reference, Identifier:String=@n40) -> (then, ReturnValue:Transform)
   n44 = BreakTransform(InTransform=@n43)
   > SetGizmoTransform(Identifier=@n40, NewTransform=MakeTransform(Location=@n42, Rotation=@n44.Rotation, Scale=@n44.Scale), bUndoable=false)
  ]}
 HandleObjsAddedRemoved {nodes [
   HandleObjsAddedRemoved() {is=title K2Node_FunctionEntry}
   > UpdateSelectedActors(BoxPos=Get SelectionBoundOrigin(), BoxExtent=Get SelectionBoundExtent())
   > OnObjectVisChanged()
  ]}
 HandleScaleGizmos {nodes [
   n1 = HandleScaleGizmos!() -> (then, IsCreate:bool=true) {is=title K2Node_FunctionEntry}
   > n2 = Branch(Condition=@n1)
   > n3 = Branch(Condition=Get Use Edge Scaling(self=Get Settings()))
   > n4 = Sequence!(execute) -> (then_0, then_1, then_2:Exec, then_3:Exec, then_4:Exec, then_5:Exec)
   n5 = BreakVector(InVec=Get SelectionBoundOrigin())
   n6 = BreakVector(InVec=Get SelectionBoundExtent())
   n7 = float + float(A=@n5, B=@n6)
   n8 = Append(A=Get BoundEdgeGizmoIdentifier() -> (BoundEdgeGizmoIdentifier:String), B="To String (Integer)"(InInt=0))
   n9 = Make Scriptable Tool Gizmo Options!(
    GizmoMode:EScriptableToolGizmoMode Enum=TranslationOnly
    TranslationParts:Bitmask Integer=131
    RotationParts:Bitmask Integer=241
    ScaleParts:Bitmask Integer=1
    CoordSystem:EScriptableToolGizmoCoordinateSystem Enum=Local
    bSnapTranslation:bool=false
    bSnapRotation:bool=false
    bRepositionable:bool=False
    bAllowNegativeScaling:bool=false
   ) -> (ScriptableToolGizmoOptions:Scriptable Tool Gizmo Options Structure) {is=title K2Node_MakeStruct}
   > CreateTRSGizmo!(
    execute
    self:Scriptable Interactive Tool Object Reference
    Identifier:String=@n8
    InitialTransform:Transform=MakeTransform(Location=MakeVector(X=@n7, Y=@n5.Y, Z=@n5.Z), Rotation="0, 0, 0")
    GizmoOptions:Scriptable Tool Gizmo Options Structure=@n9
   ) -> (Success:Exec, Failure:Exec)
   > Add Unique(TargetArray=Get AllGizmoIdentifiers(), NewItem=@n8)
   n10 = BreakVector(InVec=Get SelectionBoundOrigin())
   n11 = BreakVector(InVec=Get SelectionBoundExtent())
   n12 = float - float(A=@n10.Y, B=@n11.Y)
   n13 = Append(A=Get BoundEdgeGizmoIdentifier(), B="To String (Integer)"(InInt=3))
   n14 = Make Scriptable Tool Gizmo Options(
    GizmoMode=TranslationOnly
    TranslationParts=133
    RotationParts=241
    ScaleParts=1
    CoordSystem=Local
    bSnapTranslation=false
    bSnapRotation=false
    bRepositionable=False
    bAllowNegativeScaling=false
   )
   CreateTRSGizmo(
    execute=@n4.then_2
    Identifier=@n13
    InitialTransform=MakeTransform(Location=MakeVector(X=@n10, Y=@n12, Z=@n10.Z), Rotation="0, 0, 0")
    GizmoOptions=@n14
   )
   > Add Unique(TargetArray=Get AllGizmoIdentifiers(), NewItem=@n13)
   n15 = float + float(A=@n10.Y, B=@n11.Y)
   n16 = Append(A=Get BoundEdgeGizmoIdentifier(), B="To String (Integer)"(InInt=2))
   CreateTRSGizmo(
    execute=@n4.then_3
    Identifier=@n16
    InitialTransform=MakeTransform(Location=MakeVector(X=@n10, Y=@n15, Z=@n10.Z), Rotation=MakeRotator(Roll=-180))
    GizmoOptions=@n14
   )
   > Add Unique(TargetArray=Get AllGizmoIdentifiers(), NewItem=@n16)
   n17 = BreakVector(InVec=Get SelectionBoundOrigin())
   n18 = BreakVector(InVec=Get SelectionBoundExtent())
   n19 = float + float(A=@n17.Z, B=@n18.Z)
   n20 = Append(A=Get BoundEdgeGizmoIdentifier(), B="To String (Integer)"(InInt=4))
   n21 = Make Scriptable Tool Gizmo Options(
    GizmoMode=TranslationOnly
    TranslationParts=137
    RotationParts=241
    ScaleParts=1
    CoordSystem=Local
    bSnapTranslation=false
    bSnapRotation=false
    bRepositionable=False
    bAllowNegativeScaling=false
   )
   CreateTRSGizmo(
    execute=@n4.then_4
    Identifier=@n20
    InitialTransform=MakeTransform(Location=MakeVector(X=@n17, Y=@n17.Y, Z=@n19), Rotation="0, 0, 0")
    GizmoOptions=@n21
   )
   > Add Unique(TargetArray=Get AllGizmoIdentifiers(), NewItem=@n20)
   n22 = float - float(A=@n17.Z, B=@n18.Z)
   n23 = Append(A=Get BoundEdgeGizmoIdentifier(), B="To String (Integer)"(InInt=5))
   CreateTRSGizmo(
    execute=@n4.then_5
    Identifier=@n23
    InitialTransform=MakeTransform(Location=MakeVector(X=@n17, Y=@n17.Y, Z=@n22), Rotation=MakeRotator(Pitch=180))
    GizmoOptions=@n21
   )
   > Add Unique(TargetArray=Get AllGizmoIdentifiers(), NewItem=@n23)
   n24 = float - float(A=@n5, B=@n6)
   n25 = Append(A=Get BoundEdgeGizmoIdentifier(), B="To String (Integer)"(InInt=1))
   CreateTRSGizmo(
    execute=@n4.then_1
    Identifier=@n25
    InitialTransform=MakeTransform(Location=MakeVector(X=@n24, Y=@n5.Y, Z=@n5.Z), Rotation=MakeRotator(Yaw=-180))
    GizmoOptions=@n9
   )
   > Add Unique(TargetArray=Get AllGizmoIdentifiers(), NewItem=@n25)
   n26 = Get BoundGizmoIdentifier() -> (BoundGizmoIdentifier:String)
   CreateTRSGizmo!(
    execute=@n3.else
    self
    Identifier=@n26
    InitialTransform=MakeTransform(Location=Get SelectionBoundOrigin(), Rotation="0, 0, 0", Scale=Get SelectionBoundExtent())
    GizmoOptions_GizmoMode:EScriptableToolGizmoMode Enum=Combined
    GizmoOptions_TranslationParts:Bitmask Integer=255
    GizmoOptions_RotationParts:Bitmask Integer=241
    GizmoOptions_ScaleParts:Bitmask Integer=255
    GizmoOptions_CoordSystem:EScriptableToolGizmoCoordinateSystem Enum=World
    GizmoOptions_bSnapTranslation:bool=false
    GizmoOptions_bSnapRotation:bool=false
    GizmoOptions_bRepositionable:bool=False
    GizmoOptions_bAllowNegativeScaling:bool=false
   ) -> (Success, Failure)
   > Add Unique(TargetArray=Get AllGizmoIdentifiers(), NewItem=@n26)
   n27 = For Each Loop(Exec=@n2.else, Array=Get AllGizmoIdentifiers())
   Clear(execute=@n27.Completed, TargetArray=Get AllGizmoIdentifiers())
   DestroyTRSGizmo!(execute=@n27.LoopBody, self:Scriptable Interactive Tool Object Reference, Identifier:String=@n27) -> (Success:Exec, Failure:Exec)
  ]}
 HideSelected {nodes [
   HideSelected() {is=title K2Node_FunctionEntry}
   > Clear(TargetArray:Array of Actor Object References=Get HiddenActorsToIgnore() -> (HiddenActorsToIgnore:Array of Actor Object References))
   > Set HiddenActorsToIgnore(HiddenActorsToIgnore:Array of Actor Object References=Get ActorsToAlwaysIgnore_FromHidden() -> (ActorsToAlwaysIgnore_FromHidden:Array of Actor Object References)) -> (Output_Get:Array of Actor Object References)
   > n1 = GetSelectedLevelActors(self=Get EditorActorSubsystem())
   > n2 = For Each Loop(Array:Array of Actor Object References=Get AllActorsAtStart() -> (AllActorsAtStart:Array of Actor Object References)) -> (Array Element:Actor Object Reference)
   > Set CurrentActorInLoop(CurrentActorInLoop:Actor Object Reference=@n2) -> (Output_Get:Actor Object Reference)
   n3 = Get Class Display Name(Class=Get Class(Object=Get CurrentActorInLoop() -> (CurrentActorInLoop:Actor Object Reference)))
   > Branch(Condition="integer < integer"(A=FindSubstring(SearchIn=@n3, Substring=PreviewGeometryActor)))
   > n4 = Branch(Condition="integer < integer"(A=Find Item!(TargetArray:Array of Actor Object References=@n1, ItemToFind:"Actor Object Reference (by ref)"=@n2) -> (ReturnValue)))
   > SetIsTemporarilyHiddenInEditor!(
    execute
    self:Actor Object Reference=Get CurrentActorInLoop()
    bIsHidden:bool=Find(
     TargetMap:Map of Actor Object References to Booleans=Get IsHiddenAtStartOfTool() -> (IsHiddenAtStartOfTool:Map of Actor Object References to Booleans)
     Key:"Actor Object Reference (by ref)"=Get CurrentActorInLoop()
    ) -> (Value:bool)
   ) -> (then)
   SetIsTemporarilyHiddenInEditor(execute=@n4.else, self=Get CurrentActorInLoop(), bIsHidden=true)
   > Add(TargetArray=Get HiddenActorsToIgnore(), NewItem=Get CurrentActorInLoop())
  ]}
 MouseWheelChange {nodes [
   n1 = MouseWheelChange!() -> (then, IsUp:bool=~) {is=title K2Node_FunctionEntry}
   > Branch(Condition=Get IsModKeyPressed())
   > n2 = Sequence!(execute) -> (then_0, then_1, then_2, then_3, then_4, then_5, then_6:Exec)
   > HandleScaleGizmos(IsCreate=false) {is=title K2Node_CallFunction}
   n3 = float * float(A=Get Transform Gizmo Sensitivity(self=Get Settings()), B=100)
   n4 = Get SelectionBoundExtent()
   n5 = BreakVector(InVec=@n4)
   n6 = "Max (Float)"!(A=@n5, B=@n5.Y, C:double=@n5.Z) -> (ReturnValue)
   n7 = vector * vector(
    A=MakeVector(X=float / float(A=@n5, B=@n6), Y=float / float(A=@n5.Y, B=@n6), Z=float / float(A=@n5.Z, B=@n6))
    B=SelectFloat(A=@n3, B=float * float(A=@n3, B=-1), bPickA=@n1)
   ) -> (ReturnValue=~)
   Set SelectionBoundExtent(
    execute=@n2.then_1
    SelectionBoundExtent=ClampVectorSize!(A:Vector=vector + vector(A=@n4, B=@n7), Min:double=0.1, Max:double=10000000000) -> (ReturnValue:Vector)
   )
   n8 = Get SelectionBoundOrigin()
   n9 = Get SelectionBoundExtent()
   UpdateBBox(execute=@n2.then_2, BoundOrigin=@n8, BoundExtent=@n9, Update Properties=false)
   n10 = WaitForTime(execute=@n2.then_3, PauseTime=0.05)
   > Branch(Condition=@n10)
   > UpdateSelectedActors(BoxPos=@n8, BoxExtent=@n9)
   HandleScaleGizmos(execute=@n2.then_5, IsCreate=true)
   OnObjectVisChanged(execute=@n2.then_6)
  ]}
 OnBBoxAppearanceChanged {nodes [
   OnBBoxAppearanceChanged!() -> (then, PropertySet:Scriptable Interactive Tool Property Set Object Reference, PropertyName:String) {is=title K2Node_FunctionEntry}
   > UpdateBBoxAppearance()
  ]}
 OnBeginHover {nodes [
   OnBeginHover!() -> (
    then
    CurrentPos_WorldRay:Ray Structure
    CurrentPos_bHas2D:bool=~
    CurrentPos_ScreenPosition:Vector 2D Structure
    Modifiers:Scriptable Tool Modifier States Structure
   ) {is=title K2Node_FunctionEntry}
   > UpdateHoverText()
  ]}
 OnClickDrag {nodes [
   n1 = OnClickDrag!() -> (
    then
    DragPos_WorldRay:Ray Structure
    DragPos_bHas2D:bool=~
    DragPos_ScreenPosition:Vector 2D Structure
    Modifiers:Scriptable Tool Modifier States Structure
    MouseButton:EScriptableToolMouseButton Enum=LeftButton
   ) {is=title K2Node_FunctionEntry}
   n2 = GetRayStartEnd!(Ray:Ray Structure=@n1, StartDistance:double=0, EndDistance:double=0) -> (StartPoint:Vector, EndPoint:Vector)
   > n3 = Line Trace By Channel!(
    execute
    Start:Vector=@n2
    End:Vector=@n2.EndPoint
    TraceChannel:ETraceTypeQuery Enum
    bTraceComplex:bool=true
    ActorsToIgnore:Array of Actor Object References
    DrawDebugType:EDrawDebugTrace Enum
    bIgnoreSelf:bool
    TraceColor:Linear Color Structure
    TraceHitColor:Linear Color Structure
    DrawTime:float
   ) -> (then, OutHit:Hit Result Structure, ReturnValue:bool)
   n4 = BreakHitResult!(Hit:"Hit Result Structure (by ref)"=Get ClickDragHitBeforeDrag() -> (ClickDragHitBeforeDrag:Hit Result Structure)) -> (
    bBlockingHit:bool
    bInitialOverlap:bool
    Time:float
    Distance:float
    Location:Vector
    ImpactPoint:Vector
    Normal:Vector
    ImpactNormal:Vector
    PhysMat:Physical Material Object Reference
    HitActor:Actor Object Reference
    HitComponent:Primitive Component Object Reference
    HitBoneName:Name
    BoneName:Name
    HitItem:int
    ElementIndex:int
    FaceIndex:int
    TraceStart:Vector
    TraceEnd:Vector
   )
   n5 = Get DragSquareLines() -> (DragSquareLines:Array of Scriptable Tool Line Object References)
   n6 = BreakHitResult(Hit=@n3)
   n7 = vector * vector!(A=vector - vector(A=@n4.Location, B=@n6.Location), B:Vector="0.5,0.5,0.5") -> (ReturnValue_X:double, ReturnValue_Y:double, ReturnValue_Z:double)
   n8 = MakeVector(X=float * float(A=@n7.ReturnValue_Y, B=-1), Y=@n7, Z=@n7.ReturnValue_Z)
   n9 = vector * vector!(A=vector + vector(A=@n4.Location, B=@n6.Location), B="0.5,0.5,0.5") -> (ReturnValue:Vector=~)
   n10 = vector + vector(A=@n9, B=@n8)
   > SetLineEndPoints!(
    execute
    self:Scriptable Tool Line Object Reference="Get (a copy)"(Array:Array of Scriptable Tool Line Object References=@n5) -> (Output:Scriptable Tool Line Object Reference)
    Start:Vector=@n4.Location
    End:Vector=@n10
   ) -> (then)
   > SetLineEndPoints(self="Get (a copy)"(Array=@n5, Dimension 1=1), Start=@n10, End=@n6.Location)
   n11 = vector - vector(A=@n8, B=@n9)
   > SetLineEndPoints(self="Get (a copy)"(Array=Get DragSquareLines(), Dimension 1=2), Start=@n6.Location, End=@n11)
   > SetLineEndPoints(self="Get (a copy)"(Array=Get DragSquareLines(), Dimension 1=3), Start=@n11, End=@n4.Location)
   > DrawDebugSphere!(
    execute
    Center:Vector=@n4.Location
    Radius:float=100
    Segments:int=12
    LineColor:Linear Color Structure=(R=1,G=0.061,B=0.061,A=1)
    Duration:float=10
    Thickness:float=0
    DepthPriority:EDrawDebugSceneDepthPriorityGroup Enum=World
   ) -> (then)
   > DrawDebugSphere(Center=@n10, Radius=100, Segments=12, LineColor=(R=0,G=0.9475,B=0.0405,A=1), Duration=10, Thickness=0, DepthPriority=World)
   > DrawDebugSphere(Center=@n6.Location, Radius=100, Segments=12, LineColor=(R=0,G=0.0545,B=0.9475,A=1), Duration=10, Thickness=0, DepthPriority=World)
   > DrawDebugSphere(Center=@n11, Radius=100, Segments=12, LineColor=(R=0.9006,G=0.1725,B=1,A=1), Duration=10, Thickness=0, DepthPriority=World)
   DrawDebugArrow!(
    execute
    LineStart:Vector="0, 0, 0"
    LineEnd:Vector="0, 0, 0"
    ArrowSize:float=0
    LineColor:Linear Color Structure=(R=0,G=0.0452,B=0.785,A=1)
    Duration:float=0
    Thickness:float=1
    DepthPriority:EDrawDebugSceneDepthPriorityGroup Enum=World
   ) -> (then)
  ]}
 OnClickPress {nodes [
   n1 = OnClickPress!() -> (
    then
    PressPos_WorldRay:Ray Structure
    PressPos_bHas2D:bool=~
    PressPos_ScreenPosition:Vector 2D Structure
    Modifiers:Scriptable Tool Modifier States Structure
    MouseButton:EScriptableToolMouseButton Enum=LeftButton
   ) {is=title K2Node_FunctionEntry}
   > n2 = Set ClickDragWorldRayBeforeDrag(ClickDragWorldRayBeforeDrag:Ray Structure=@n1) -> (Output_Get:Ray Structure)
   n3 = GetRayStartEnd(Ray=@n2, StartDistance=0, EndDistance=0)
   > n4 = Line Trace By Channel(Start=@n3, End=@n3.EndPoint, bTraceComplex=true)
   > Set ClickDragHitBeforeDrag(ClickDragHitBeforeDrag:Hit Result Structure=@n4) -> (Output_Get:Hit Result Structure)
  ]}
 OnCycleObjVisibility {nodes [
   OnCycleObjVisibility!() -> (then, Key:Key Structure=None, Modifiers:Scriptable Tool Modifier States Structure) {is=title K2Node_FunctionEntry}
   n1 = Get Settings()
   > n2 = Switch on E_3DSelect_HidingOptions!(
    execute
    Selection:E_3DSelect_HidingOptions Enum=Get Object Visibility!(self:BP STPS 3DSelect Object Reference=@n1) -> (Object Visibility:E_3DSelect_HidingOptions Enum=NewEnumerator0)
   ) -> (NewEnumerator0:Exec, NewEnumerator1:Exec, NewEnumerator2:Exec) {is=title K2Node_SwitchEnum}
   > Set Object Visibility(Object Visibility=NewEnumerator1, self=@n1) -> (Output_Get=NewEnumerator0)
   Set Object Visibility(execute=@n2.NewEnumerator1, Object Visibility=NewEnumerator2, self=@n1) -> (Output_Get=NewEnumerator0)
   Set Object Visibility(execute=@n2.NewEnumerator2, Object Visibility=NewEnumerator0, self=@n1) -> (Output_Get=NewEnumerator0)
  ]}
 OnExcludeClassKeyPressed {nodes [
   OnExcludeClassKeyPressed!() -> (then, Key:Key Structure=None, Modifiers:Scriptable Tool Modifier States Structure) {is=title K2Node_FunctionEntry}
   > Add Unique(
    TargetArray=Get Exclude Class!(self:BP STPS 3DSelect Object Reference=Get Settings()) -> (Exclude Class:Array of Strings)
    NewItem=Get Class Line Value Class String() -> (Class Line Value Class String:String)
   )
  ]}
 OnExcludeStringKeyPressed {nodes [
   OnExcludeStringKeyPressed!() -> (then, Key:Key Structure=None, Modifiers:Scriptable Tool Modifier States Structure) {is=title K2Node_FunctionEntry}
   > Add Unique(
    TargetArray=Get Exclude Label!(self:BP STPS 3DSelect Object Reference=Get Settings()) -> (Exclude Label:Array of Strings)
    NewItem="To String (Text)"(InText=Get Actor Line Value())
   )
  ]}
 OnFilterChanged {nodes [
   OnFilterChanged!() -> (then, PropertySet:Scriptable Interactive Tool Property Set Object Reference, PropertyName:String) {is=title K2Node_FunctionEntry}
   > UpdateFilterVariables()
   > UpdateSelectedActors(BoxPos=Get SelectionBoundOrigin(), BoxExtent=Get SelectionBoundExtent())
   > OnObjectVisChanged()
  ]}
 OnFilterCycleKeyPressed {nodes [
   OnFilterCycleKeyPressed!() -> (then, Key:Key Structure=None, Modifiers:Scriptable Tool Modifier States Structure) {is=title K2Node_FunctionEntry}
   > n1 = Switch on E_3DSelect_FilterOptions(Selection=Get Filter Mode(self=Get Settings()) -> (Filter Mode=NewEnumerator0))
   > Set Filter Mode!(execute, Filter Mode:E_3DSelect_FilterOptions Enum=NewEnumerator1, self:BP STPS 3DSelect Object Reference=Get Settings()) -> (then, Output_Get:E_3DSelect_FilterOptions Enum=NewEnumerator0)
   Set Filter Mode(execute=@n1.NewEnumerator1, Filter Mode=NewEnumerator2, self=Get Settings()) -> (Output_Get=NewEnumerator0)
   Set Filter Mode(execute=@n1.NewEnumerator2, Filter Mode=NewEnumerator0, self=Get Settings()) -> (Output_Get=NewEnumerator0)
  ]}
 OnHideSelectedChanged {nodes [
   n1 = OnHideSelectedChanged!() -> (then, PropertySet:Scriptable Interactive Tool Property Set Object Reference, PropertyName:String, bNewValue:bool=~) {is=title K2Node_FunctionEntry}
   > n2 = Branch(Condition=@n1.bNewValue)
   > n3 = HideSelected() {is=title K2Node_CallFunction}
   UpdateSelectionBounds(execute=[@n4.then, @n3.then], ForceUpdate=false)
   n4 = ShowAll(execute=@n2.else)
  ]}
 OnHitByClick {nodes [
   n1 = OnHitByClick!() -> (
    then
    ClickPos:Input Device Ray Structure
    Modifiers_bShiftDown:bool=~
    Modifiers_bCtrlDown:bool=~
    Modifiers_bAltDown:bool=~
    MouseButton:EScriptableToolMouseButton Enum=LeftButton
   ) {is=title K2Node_FunctionEntry}
   > n2 = Branch(Condition=Get IsModKeyPressed())
   n3 = GetSelectedLevelActors(execute=@n2.else, self=Get EditorActorSubsystem())
   > n4 = Set SelLevelActorsBeforeClickAction(SelLevelActorsBeforeClickAction:Array of Actor Object References=@n3) -> (Output_Get:Array of Actor Object References)
   > Remove Item!(execute, TargetArray:Array of Actor Object References=@n4, Item:"Actor Object Reference (by ref)"=Get FadeCubeActor()) -> (then, ReturnValue:bool)
   > Set Hit Actor(Hit Actor:Actor Object Reference=BreakHitResult(Hit=Get SingleClickHit() -> (SingleClickHit:Hit Result Structure)).HitActor) -> (Output_Get:Actor Object Reference)
   > n5 = Branch(Condition=AND Boolean!(A=NOT Boolean(A=@n1.Modifiers_bShiftDown), B=NOT Boolean(A=@n1.Modifiers_bCtrlDown), C:bool=NOT Boolean(A=@n1.Modifiers_bAltDown)) -> (ReturnValue))
   n6 = Make Array("[0]":Actor Object Reference=Get Hit Actor() -> (Hit Actor:Actor Object Reference)) -> (Array:Array of Actor Object References)
   > Remove Item(TargetArray=@n6, Item=Get FadeCubeActor())
   > n7 = SetSelectedLevelActors_3DSel!(execute, self, ActorsToSelect:Array of Actor Object References=@n6) -> (then, FilteredActorsKept:Array of Actor Object References, FilteredActorsRemoved:Array of Actor Object References)
   > Branch(Condition="integer > integer"(A=Length(TargetArray=@n7.FilteredActorsRemoved)))
   n8 = Get Hit Actor()
   > DisplayUserWarningMessage(Message=Format Text!(
     Format=Get SelectionOfActorRemovedByFilter_WarningMessage() -> (SelectionOfActorRemovedByFilter_WarningMessage:Text)
     0:String=GetActorLabel(self=@n8, bCreateIfNone=true)
    ) -> (Result))
   > Set Last Clear Time(Last Clear Time=Get CurTime())
   > SetSelectedLevelActors_3DSel(ActorsToSelect=Get SelLevelActorsBeforeClickAction() -> (SelLevelActorsBeforeClickAction:Array of Actor Object References))
   n9 = Get!() -> (Modifiers_bShiftDown:bool=~, Modifiers_bCtrlDown:bool=~, Modifiers_bAltDown:bool=~)
   n10 = Branch(execute=@n5.else, Condition=AND Boolean(A=@n9, B=NOT Boolean(A=@n9.Modifiers_bCtrlDown), C=NOT Boolean(A=@n9.Modifiers_bAltDown)))
   > Add Unique(
    TargetArray:Array of Actor Object References=Get SelLevelActorsBeforeClickAction()
    NewItem:"Actor Object Reference (by ref)"=Get Hit Actor()
   )
   > Remove Item(TargetArray=Get SelLevelActorsBeforeClickAction(), Item=Get FadeCubeActor())
   > n11 = SetSelectedLevelActors_3DSel(ActorsToSelect=Get SelLevelActorsBeforeClickAction())
   > Branch(Condition="integer > integer"(A=Length(TargetArray=@n11.FilteredActorsRemoved)))
   n12 = Get Hit Actor()
   > DisplayUserWarningMessage(Message=Format Text(Format=Get SelectionOfActorRemovedByFilter_WarningMessage(), 0=GetActorLabel(self=@n12, bCreateIfNone=true)))
   > Set Last Clear Time(Last Clear Time=Get CurTime())
   > SetSelectedLevelActors_3DSel(ActorsToSelect=Get SelLevelActorsBeforeClickAction())
   n13 = Get() -> (Modifiers_bShiftDown=~, Modifiers_bCtrlDown=~, Modifiers_bAltDown=~)
   Branch(execute=@n10.else, Condition=AND Boolean(A=NOT Boolean(A=@n13), B=@n13.Modifiers_bCtrlDown, C=NOT Boolean(A=@n13.Modifiers_bAltDown)))
   n14 = Find Item(TargetArray=Get SelLevelActorsBeforeClickAction(), ItemToFind=Get Hit Actor())
   > Branch(Condition="integer > integer"(A=@n14))
   > Remove Index!(execute, TargetArray:Array of Actor Object References=Get SelLevelActorsBeforeClickAction(), IndexToRemove:int=@n14) -> (then)
   > SetSelectedLevelActors_3DSel(ActorsToSelect=Get SelLevelActorsBeforeClickAction())
  ]}
 OnHotkeyInterrupt {nodes [
   OnHotkeyInterrupt() {is=title K2Node_FunctionEntry}
   > Branch(Condition=Get IsModKeyPressed())
   > OnModKeyReleased!(execute, self, Key:Key Structure=None, Modifiers:Scriptable Tool Modifier States Structure) -> (then)
  ]}
 OnHoverEnded {nodes [
   OnHoverEnded() {is=title K2Node_FunctionEntry}
   > "SetText (Text)"(self=Get Value!(self:BP UW Readout Line Object Reference=Get Readout Line Actor() -> (Readout Line Actor:BP UW Readout Line Object Reference)) -> (Value:Common Text Object Reference))
   > "SetText (Text)"(self=Get Value(self=Get Readout Line Class() -> (Readout Line Class:BP UW Readout Line Object Reference)))
  ]}
 OnHoverSequenceHitTest {nodes [
   n1 = OnHoverSequenceHitTest!() -> (
    then
    CurrentPos_WorldRay:Ray Structure
    CurrentPos_bHas2D:bool=~
    CurrentPos_ScreenPosition:Vector 2D Structure
    Modifiers:Scriptable Tool Modifier States Structure
   ) {is=title K2Node_FunctionEntry}
   > SetFocusInViewport(self:Scriptable Interactive Tool Object Reference)
   n2 = GetRayStartEnd(Ray=@n1, StartDistance=0, EndDistance=0)
   > n3 = Line Trace By Channel(Start=@n2, End=@n2.EndPoint, bTraceComplex=true, ActorsToIgnore=Get HiddenActorsToIgnore())
   > Set HoverOutHit(HoverOutHit:Hit Result Structure=@n3) -> (Output_Get:Hit Result Structure)
   > Return Node!(
    execute
    ReturnValue_bHit:bool=true
    ReturnValue_HitDepth:double=340282346638528859811704183484516925440
    ReturnValue_HitNormal:Vector="0,0,1"
    ReturnValue_bHasHitNormal:bool=False
    ReturnValue_HitIdentifier:int=0
    ReturnValue_HitObject:Object Reference
   )
  ]}
 OnHoverUpdated {nodes [
   n1 = OnHoverUpdated!() -> (
    then
    CurrentPos_WorldRay:Ray Structure
    CurrentPos_bHas2D:bool=~
    CurrentPos_ScreenPosition_X:double
    CurrentPos_ScreenPosition_Y:double
    Modifiers_bShiftDown:bool=~
    Modifiers_bCtrlDown:bool=~
    Modifiers_bAltDown:bool=~
   ) {is=title K2Node_FunctionEntry}
   n2 = GetRayStartEnd(Ray=@n1, StartDistance=0, EndDistance=0)
   > Line Trace By Channel(Start=@n2, End=@n2.EndPoint, bTraceComplex=true)
   > Return Node!(execute, ReturnValue:bool=false)
  ]}
 OnIfHitByClick {nodes [
   n1 = OnIfHitByClick!() -> (
    then
    ClickPos_WorldRay:Ray Structure
    ClickPos_bHas2D:bool=~
    ClickPos_ScreenPosition:Vector 2D Structure
    MouseButton:EScriptableToolMouseButton Enum=LeftButton
   ) {is=title K2Node_FunctionEntry}
   n2 = GetRayStartEnd(Ray=@n1, StartDistance=0, EndDistance=0)
   > n3 = Line Trace By Channel(Start=@n2, End=@n2.EndPoint, bTraceComplex=true, ActorsToIgnore=Get HiddenActorsToIgnore())
   > Set SingleClickHit(SingleClickHit:Hit Result Structure=@n3) -> (Output_Get:Hit Result Structure)
   > Return Node!(
    execute
    ReturnValue_bHit:bool=@n3.ReturnValue
    ReturnValue_HitDepth:double=340282346638528859811704183484516925440
    ReturnValue_HitNormal:Vector="0,0,1"
    ReturnValue_bHasHitNormal:bool=False
    ReturnValue_HitIdentifier:int=0
    ReturnValue_HitObject:Object Reference
   )
  ]}
 OnIncludeClassKeyPressed {nodes [
   OnIncludeClassKeyPressed!() -> (then, Key:Key Structure=None, Modifiers:Scriptable Tool Modifier States Structure) {is=title K2Node_FunctionEntry}
   > Add Unique(
    TargetArray:Array of Strings=Get Include Class!(self:BP STPS 3DSelect Object Reference=Get Settings()) -> (Include Class:Array of Strings)
    NewItem:"String (by ref)"=Get Class Line Value Class String()
   )
  ]}
 OnIncludeKeyPressed {nodes [
   n1 = OnIncludeKeyPressed!() -> (then, Key:Key Structure=None, Modifiers:Scriptable Tool Modifier States Structure) {is=title K2Node_FunctionEntry}
   > n2 = Branch(Condition=Break Scriptable Tool Modifier States!(ScriptableToolModifierStates:"Scriptable Tool Modifier States Structure (by ref)"=@n1.Modifiers) -> (bShiftDown:bool=~, bCtrlDown:bool=~, bAltDown:bool=~) {is=title K2Node_BreakStruct})
   n3 = Get Settings()
   > Add Unique(
    TargetArray=Get Include Label!(self:BP STPS 3DSelect Object Reference=@n3) -> (Include Label:Array of Strings)
    NewItem="To String (Text)"(InText=Get Actor Line Value())
   )
   Add Unique(execute=@n2.else, TargetArray=Get Include Class(self=@n3), NewItem="To String (Text)"(InText=Get ClassLineValue()))
  ]}
 OnModKeyPressed {nodes [
   OnModKeyPressed!() -> (then, Key:Key Structure=None, Modifiers:Scriptable Tool Modifier States Structure) {is=title K2Node_FunctionEntry}
   > n1 = Branch(Condition=Get IsModKeyPressed())
   n2 = Sequence!(execute=@n1.else) -> (then_0, then_1, then_2)
   > Set IsModKeyPressed(IsModKeyPressed:bool=true) -> (Output_Get:bool)
   > Set SuspendSelChangeDetection(SuspendSelChangeDetection:bool=true) -> (Output_Get:bool)
   > UpdateSelectionBounds(ForceUpdate=false)
   HandleScaleGizmos(execute=@n2.then_1, IsCreate=true)
   ControlFadeCube(execute=@n2.then_2) {is=title K2Node_CallFunction}
   > SetFadeCubeSMTransform!(execute, self, Location:Vector=Get SelectionBoundOrigin(), Extent:Vector=Get SelectionBoundExtent()) -> (then)
  ]}
 OnModKeyReleased {nodes [
   OnModKeyReleased!() -> (then, Key:Key Structure=None, Modifiers:Scriptable Tool Modifier States Structure) {is=title K2Node_FunctionEntry}
   > Set IsModKeyPressed(IsModKeyPressed=false)
   > Set SuspendSelChangeDetection(SuspendSelChangeDetection=false)
   > HandleScaleGizmos(IsCreate=false)
   > n1 = Branch(Condition=Get Show Bounding Box Mesh(self=Get Settings()))
   Destroy Actor(execute=@n1.else, self=Get FadeCubeActor())
  ]}
 OnMouseWheelDown {nodes [
   OnMouseWheelDown!() -> (then, CurrentPos:Input Device Ray Structure, Modifiers:Scriptable Tool Modifier States Structure) {is=title K2Node_FunctionEntry}
   > MouseWheelChange!(execute, self, IsUp:bool=false) -> (then) {is=title K2Node_CallFunction}
  ]}
 OnMouseWheelUp {nodes [
   OnMouseWheelUp!() -> (then, CurrentPos:Input Device Ray Structure, Modifiers:Scriptable Tool Modifier States Structure) {is=title K2Node_FunctionEntry}
   > MouseWheelChange(IsUp=true)
  ]}
 OnObjectVisChanged {nodes [
   OnObjectVisChanged!() -> (then, PropertySet:Scriptable Interactive Tool Property Set Object Reference, PropertyName:String) {is=title K2Node_FunctionEntry}
   > n1 = Switch on E_3DSelect_HidingOptions(Selection=Get Object Visibility(self=Get Settings()) -> (Object Visibility=NewEnumerator0))
   > n2 = ShowAll()
   UpdateSelectionBounds(execute=[@n2.then, @n3.then, @n4.then], ForceUpdate=false)
   n3 = ShowOnlySelected(execute=@n1.NewEnumerator1)
   n4 = HideSelected(execute=@n1.NewEnumerator2)
  ]}
 OnObjectsDropped {nodes [
   n1 = PrintString!(execute, InString=START Dropped Objects, bPrintToScreen, bPrintToLog, TextColor, Duration, Key) -> (then)
   n2 = OnObjectsDropped!() -> (then, DroppedObjects:Array of Object References, DroppedActors:Array of Actor Object References) {is=title K2Node_FunctionEntry}
   n3 = For Each Loop(Exec=@n1.then, Array:Array of Object References=@n2) -> (Array Element:Object Reference)
   PrintString(execute=@n3.Completed, InString=START Dropped Actors)
   > n4 = For Each Loop(Array:Array of Actor Object References=@n2.DroppedActors) -> (Array Element:Actor Object Reference)
   HandleObjsAddedRemoved(execute=[@n2.then, @n4.Completed]) {is=title K2Node_CallFunction}
   PrintString(execute=@n4.LoopBody, InString=GetDisplayName(Object=@n4))
   PrintString(execute=@n3.LoopBody, InString=GetDisplayName(Object=@n3))
   Return Node!(execute)
  ]}
 OnObjectsPlaced {nodes [
   n1 = PrintString(InString=START Placed Actors)
   n2 = OnObjectsPlaced!() -> (then, ObjToUse:Object Reference, PlacedActors:Array of Actor Object References) {is=title K2Node_FunctionEntry}
   PrintString(execute=@n1.then, InString=GetDisplayName(Object=@n2))
   > n3 = For Each Loop(Array:Array of Object References=@n2.PlacedActors) -> (Array Element:Object Reference)
   HandleObjsAddedRemoved(execute=[@n2.then, @n3.Completed])
   PrintString(execute=@n3.LoopBody, InString=GetDisplayName(Object=@n3))
   Return Node()
  ]}
 OnPasteBegin {nodes [
   OnPasteBegin() {is=title K2Node_FunctionEntry}
   > PrintString(InString=PASTE BEGIN)
  ]}
 OnPasteEnd {nodes [
   OnPasteEnd() {is=title K2Node_FunctionEntry}
   > HandleObjsAddedRemoved()
  ]}
 OnSelectionChanged {nodes [
   n1 = OnSelectionChanged!() -> (then, NewSelection:Array of Actor Object References) {is=title K2Node_FunctionEntry}
   > Set CurNumSelectedLevelActors(CurNumSelectedLevelActors:int=Length(TargetArray=@n1)) -> (Output_Get:int=0)
   > SnapBoundsToSelection()
   > OnObjectVisChanged() {is=title K2Node_CallFunction}
  ]}
 OnSelectionSetChanged {nodes [
   OnSelectionSetChanged!() -> (then, SelectionSet:Typed Element Selection Set Object Reference) {is=title K2Node_FunctionEntry}
   > n1 = GetSelectedLevelActors(self=Get EditorActorSubsystem())
   > OnSelectionChanged(NewSelection=@n1) {is=title K2Node_CallFunction}
   > OnObjectVisChanged()
   > UpdateSelectedActorText()
  ]}
 OnShowFadeCubeChanged {nodes [
   OnShowFadeCubeChanged!() -> (then, PropertySet:Scriptable Interactive Tool Property Set Object Reference, PropertyName:String, bNewValue:bool=~) {is=title K2Node_FunctionEntry}
   > ControlFadeCube()
   > UpdateBBox(BoundOrigin=Get SelectionBoundOrigin(), BoundExtent=Get SelectionBoundExtent(), Update Properties=false)
  ]}
 OnShowOnlySelectedChanged {nodes [
   n1 = OnShowOnlySelectedChanged!() -> (then, PropertySet:Scriptable Interactive Tool Property Set Object Reference, PropertyName:String, bNewValue:bool=~) {is=title K2Node_FunctionEntry}
   > n2 = Branch(Condition=@n1.bNewValue)
   > n3 = ShowOnlySelected()
   UpdateSelectionBounds(execute=[@n3.then, @n4.then], ForceUpdate=false)
   n4 = ShowAll(execute=@n2.else)
  ]}
 OnUseFiltersToggleChanged {nodes [
   OnUseFiltersToggleChanged!() -> (then, PropertySet:Scriptable Interactive Tool Property Set Object Reference, PropertyName:String) {is=title K2Node_FunctionEntry}
   > n1 = Sequence!(execute) -> (then_0, then_1)
   > n2 = Switch on E_3DSelect_FilterOptions(Selection=Get Filter Mode(self=Get Settings()) -> (Filter Mode=NewEnumerator0))
   > n3 = SetPropertyAsReadOnly!(
    execute
    self:Scriptable Interactive Tool Property Set Object Reference=Get Settings()
    PropertyName:String=Exclude Label
    bReadOnly:bool=true
   ) -> (then, ReturnValue:Scriptable Interactive Tool Property Set Object Reference)
   > n4 = SetPropertyAsReadOnly(self=@n3, PropertyName=Exclude Class, bReadOnly=true)
   > n5 = SetPropertyAsReadOnly(self=@n4, PropertyName=Include Label, bReadOnly=true)
   > SetPropertyAsReadOnly(self=@n5, PropertyName=Include Class, bReadOnly=true)
   n6 = SetPropertyAsReadOnly(execute=@n2.NewEnumerator1, self=Get Settings(), PropertyName=Exclude Label, bReadOnly=true)
   > n7 = SetPropertyAsReadOnly(self=@n6, PropertyName=Exclude Class, bReadOnly=true)
   > n8 = SetPropertyAsReadOnly(self=@n7, PropertyName=Include Label, bReadOnly=false)
   > SetPropertyAsReadOnly(self=@n8, PropertyName=Include Class, bReadOnly=false)
   n9 = SetPropertyAsReadOnly(execute=@n2.NewEnumerator2, self=Get Settings(), PropertyName=Exclude Label, bReadOnly=false)
   > n10 = SetPropertyAsReadOnly(self=@n9, PropertyName=Exclude Class, bReadOnly=false)
   > n11 = SetPropertyAsReadOnly(self=@n10, PropertyName=Include Label, bReadOnly=true)
   > SetPropertyAsReadOnly(self=@n11, PropertyName=Include Class, bReadOnly=true)
   UpdateSelectedActors(execute=@n1.then_1, BoxPos=Get SelectionBoundOrigin(), BoxExtent=Get SelectionBoundExtent())
   > OnObjectVisChanged()
  ]}
 RefreshBoundsContents {nodes [
   RefreshBoundsContents() {is=title K2Node_FunctionEntry}
   > UpdateSelectionBounds(ForceUpdate=true)
   > UpdateSelectedActors(BoxPos=Get SelectionBoundOrigin(), BoxExtent=Get SelectionBoundExtent())
  ]}
 SetBoxEdgeWidgets {nodes [
   n1 = SetBoxEdgeWidgets!() -> (then, Positive Side:bool=true, X:bool=true, Y:bool=~, Z:bool=~) {is=title K2Node_FunctionEntry}
   n2 = BreakVector(InVec=Get SelectionBoundOrigin())
   n3 = BreakVector(InVec=Get SelectionBoundExtent())
   n4 = MakeVector(X=float + float(A=@n2, B=@n3), Y=@n2.Y, Z=@n2.Z)
   > SetGizmoTransform(Identifier=Gizmo1, NewTransform=MakeTransform(Location=@n4, Rotation="0, 0, 0"), bUndoable=false)
   SelectVector(A=@n4, B="0, 0, 0", bPickA=AND Boolean!(A=@n1, B=@n1.X) -> (ReturnValue))
   MakeVector(X=float - float(A=@n2, B=@n3), Y=@n3.Y, Z=@n3.Z)
  ]}
 SetFadeCubeSMTransform {nodes [
   n1 = SetFadeCubeSMTransform!() -> (then, Location:Vector=~, Extent:Vector=~) {is=title K2Node_FunctionEntry}
   > n2 = Is Valid(InputObject=Get FadeCubeActor())
   Return Node(execute=[@n5.then, @n2.Is Not Valid])
   n3 = MakeBoxFromCenterExtents(Center=@n1, Extents=@n1.Extent)
   n4 = "Get Size (Box)"!(InBox:"Box Structure (by ref)"=@n3) -> (ReturnValue:Vector) {is=title K2Node_CallFunction}
   > n5 = Set World Transform!(
    execute
    self:Scene Component Object Reference=Get FadeCubeActorStaticMeshComponent()
    NewTransform:"Transform (by ref)"=MakeTransform(
     Location="Get Center (Box)"!(InBox:"Box Structure (by ref)"=@n3) -> (ReturnValue:Vector) {is=title K2Node_CallFunction}
     Rotation="0, 0, 0"
     Scale=vector / vector(A:Vector=@n4, B:Vector="256,256,256") -> (ReturnValue:Vector)
    )
    bSweep:bool=false
    bTeleport:bool=false
   ) -> (then, SweepHitResult:Hit Result Structure) {is=title K2Node_CallFunction}
  ]}
 SetLinBBoxEdge {nodes [
   n1 = SetLinBBoxEdge!() -> (then, Line Number:int, StartPointNum:int, EndPointNum:int) {is=title K2Node_FunctionEntry}
   > SetLineEndPoints(
    self="Get (a copy)"(Array=Get BoxLines() -> (BoxLines:Array of Scriptable Tool Line Object References), Dimension 1=@n1)
    Start="Get (a ref)"(Array:Array of Vectors=Get CurrentCorners() -> (CurrentCorners:Array of Vectors), Dimension 1=@n1.StartPointNum) -> (Output:"Vector (by ref)")
    End="Get (a ref)"(Array=Get CurrentCorners(), Dimension 1=@n1.EndPointNum)
   )
  ]}
 SetSelectedLevelActors_3DSel {nodes [
   n1 = SetSelectedLevelActors_3DSel!() -> (then, ActorsToSelect:Array of Actor Object References) {is=title K2Node_FunctionEntry}
   > n2 = ApplyFilters!(execute, self, UnfilteredActors:Array of Actor Object References=@n1) -> (then, FilteredActorsKept:Array of Actor Object References, FilteredActorsRemoved:Array of Actor Object References) {is=title K2Node_CallFunction}
   > SetSelectedLevelActors(self=Get EditorActorSubsystem(), ActorsToSelect=@n2)
   > n3 = Set CurSelectedLevelActors(CurSelectedLevelActors:Array of Actor Object References=@n2) -> (Output_Get:Array of Actor Object References)
   > Set CurNumSelectedLevelActors(CurNumSelectedLevelActors=Length(TargetArray=@n3)) -> (Output_Get=0)
   > Return Node!(
    execute
    FilteredActorsKept:Array of Actor Object References=@n2
    FilteredActorsRemoved:Array of Actor Object References=@n2.FilteredActorsRemoved
   )
  ]}
 Setup {nodes [
   n1 = Get EditorActorSubsystem()
   Bind Event to On New Actors Dropped(self:Editor Actor Subsystem Object Reference=@n1, Delegate=Create Event()) {is=title K2Node_AddDelegate}
   > Bind Event to On New Actors Placed(self:Editor Actor Subsystem Object Reference=@n1, Delegate=Create Event()) {is=title K2Node_AddDelegate}
   > Bind Event to On Edit Paste Actors End(self:Editor Actor Subsystem Object Reference=@n1, Delegate=Create Event()) {is=title K2Node_AddDelegate}
   Setup() {is=title K2Node_FunctionEntry}
   > n2 = Sequence!(execute) -> (
    then_0
    then_1
    then_2:Exec
    then_3:Exec
    then_4:Exec
    then_5:Exec
    then_6:Exec
    then_7:Exec
    then_8:Exec
    then_9:Exec
    then_10:Exec
    then_11:Exec
    then_12:Exec
    then_13:Exec
    then_14:Exec
    then_15:Exec
   )
   K2Node_Composite_0 = Setup Property Sets(execute=@n2.then_1) {is=title K2Node_Composite}
   n3 = Construct Object from Class(
    execute=@n2.then_12
    Class=&BP_UW_SharedHeaderWidget_C
   )
   > n4 = Set Header Widget(Header Widget:BP UW Shared Header Widget Object Reference=@n3) -> (Output_Get:BP UW Shared Header Widget Object Reference)
   > SetToolPanelToolHeaderWidget(Widget=@n4)
   OnUseFiltersToggleChanged!(execute=@n2.then_13, self, PropertySet:Scriptable Interactive Tool Property Set Object Reference, PropertyName:String) -> (then) {is=title K2Node_CallFunction}
   SetFocusInViewport(execute=@n2.then_14)
   n5 = GetSelectionSet(execute=@n2.then_15, self=Get LevelEditorSubsystem())
   > Bind Event to On Selection Change(self:Typed Element Selection Set Object Reference=@n5, Delegate=Create Event()) {is=title K2Node_AddDelegate}
   K2Node_Composite_2 = Add Behaviours(execute=@n2.then_2) {is=title K2Node_Composite}
   K2Node_Composite_3 = Add Line Sets(execute=@n2.then_3) {is=title K2Node_Composite}
   K2Node_Composite_4 = Snapshot Actor Info(execute=@n2.then_4) {is=title K2Node_Composite}
   K2Node_Composite_6 = Settings Property Watches(execute=@n2.then_5) {is=title K2Node_Composite}
   K2Node_Composite_7 = Other Settings Property Watches(execute=@n2.then_6) {is=title K2Node_Composite}
   K2Node_Composite_9 = Init Sel Changed Detection(execute=@n2.then_7) {is=title K2Node_Composite}
   ControlFadeCube(execute=@n2.then_8)
   > RefreshBoundsContents() {is=title K2Node_CallFunction}
  ]}
 Setup Property Sets {
  path=Setup.K2Node_Composite_0.Setup Property Sets
  nodes [
   Inputs()
   > n1 = Sequence!(execute) -> (then_0, then_1, then_2)
   > n2 = AddPropertySetOfType(
    PropertySetType=&/BP_STPS_3DSelect.BP_STPS_3DSelect_C
    Identifier=Get PropSetSettingsIdentifier() -> (PropSetSettingsIdentifier:String)
   )
   > n3 = Cast To BP_STPS_3DSelect!(execute, Object=@n2) -> (then, CastFailed, AsBP STPS 3DSelect:BP STPS 3DSelect Object Reference) {is=title K2Node_DynamicCast}
   > n4 = Set Settings(Settings:BP STPS 3DSelect Object Reference=@n3) -> (Output_Get:BP STPS 3DSelect Object Reference)
   > RestorePropertySetSettings(PropertySet=@n4, SaveKey=Get PropSetSaveKey())
   n5 = AddPropertySetOfType(
    execute=@n1.then_1
    PropertySetType=/Game/Blueprints/Blutilities/ToolsWrkshp/ScrTools/Shared/SharedPropertySets/BP_STPS_SharedWidgetState.BP_STPS_SharedWidgetState_C
    Identifier=Get SharedStateIdentifier() -> (SharedStateIdentifier:String)
   )
   > n6 = Cast To BP_STPS_SharedWidgetState!(execute, Object=@n5) -> (then, CastFailed, AsBP STPS Shared Widget State:BP STPS Shared Widget State Object Reference) {is=title K2Node_DynamicCast}
   > Set SharedState Settings(SharedState Settings:BP STPS Shared Widget State Object Reference=@n6) -> (Output_Get:BP STPS Shared Widget State Object Reference)
   > SetPropertySetVisibleByName!(execute, self:Scriptable Interactive Tool Object Reference, Identifier:String=Get SharedStateIdentifier(), bVisible:bool=false) -> (then)
   > RestorePropertySetSettings(PropertySet=Get SharedState Settings(), SaveKey=Get PropSetSaveKey())
   Outputs()
  ]
 }
 Add Behaviours {
  path=Setup.K2Node_Composite_2.Add Behaviours
  nodes [
   Inputs()
   > n1 = Sequence!(execute) -> (then_0, then_1)
   > AddSingleKeyInputBehavior!(
    execute
    self:Scriptable Modular Behavior Tool Object Reference
    OnKeyPressed:Delegate=Create Event()
    OnKeyReleased:Delegate=Create Event()
    OnForceEndCaptureFuncIn:Delegate=Create Event()
    Key:Key Structure=Get ModKey() -> (ModKey:Key Structure)
    CaptureCheck:Delegate
    CapturePriority:int=0
   ) -> (then)
   > AddSingleKeyInputBehavior(OnKeyPressed=Create Event(), Key=Get CycleObjVisibilityKey() -> (CycleObjVisibilityKey:Key Structure), CapturePriority=0)
   n2 = Create Event()
   > AddSingleKeyInputBehavior(OnKeyPressed=Create Event(), Key=Get ExcludeClassKey() -> (ExcludeClassKey:Key Structure), CaptureCheck=@n2, CapturePriority=100)
   n3 = Create Event()
   > AddSingleKeyInputBehavior(OnKeyPressed=Create Event(), Key=Get ExcludeStringKey() -> (ExcludeStringKey:Key Structure), CaptureCheck=@n3, CapturePriority=100)
   n4 = Create Event()
   > AddSingleKeyInputBehavior(OnKeyPressed=@n4, Key=Get IncludeClassKey() -> (IncludeClassKey:Key Structure), CaptureCheck=@n2, CapturePriority=100)
   > AddSingleKeyInputBehavior(OnKeyPressed=@n4, Key=Get IncludeStringKey() -> (IncludeStringKey:Key Structure), CaptureCheck=@n3, CapturePriority=100)
   > AddSingleKeyInputBehavior(OnKeyPressed=Create Event(), Key=Get CycleFilterKey() -> (CycleFilterKey:Key Structure), CapturePriority=100)
   AddMouseHoverBehavior!(
    execute=@n1.then_1
    self:Scriptable Modular Behavior Tool Object Reference
    BeginHoverSequenceHitTest:Delegate=Create Event()
    OnBeginHover:Delegate=Create Event()
    OnUpdateHover:Delegate=Create Event()
    OnEndHover:Delegate=Create Event()
    HoverCaptureCheck:Delegate
    CapturePriority:int=100
   ) -> (then)
   > AddMouseWheelBehavior!(
    execute
    self:Scriptable Modular Behavior Tool Object Reference
    TestShouldRespondToMouseWheel:Delegate=Create Event()
    OnMouseWheelScrollUp:Delegate=Create Event()
    OnMouseWheelScrollDown:Delegate=Create Event()
    CaptureCheck:Delegate=Create Event()
    CapturePriority:int=-100
   ) -> (then)
   Outputs()
  ]
 }
 Add Line Sets {
  path=Setup.K2Node_Composite_3.Add Line Sets
  nodes [
   Inputs()
   > n1 = AddLineSet!(execute, self:Scriptable Interactive Tool Object Reference) -> (then, ReturnValue:Scriptable Tool Line Set Object Reference)
   > Set BboxLineSet(BboxLineSet:Scriptable Tool Line Set Object Reference=@n1) -> (Output_Get:Scriptable Tool Line Set Object Reference)
   > n2 = For Loop(LastIndex=11)
   UpdateBBoxAppearance(execute=@n2.Completed)
   > n3 = AddLineSet()
   > Set DragSquareLineSet(DragSquareLineSet:Scriptable Tool Line Set Object Reference=@n3) -> (Output_Get:Scriptable Tool Line Set Object Reference)
   > For Loop(LastIndex=3)
   > n4 = AddLine!(
    execute
    self:Scriptable Tool Line Set Object Reference=Get DragSquareLineSet() -> (DragSquareLineSet:Scriptable Tool Line Set Object Reference)
   ) -> (then, ReturnValue:Scriptable Tool Line Object Reference)
   > SetLineColor!(execute, self:Scriptable Tool Line Object Reference=@n4, Color_B:Byte=255, Color_G:Byte=255, Color_R:Byte=255, Color_A:Byte=255) -> (then)
   > SetLineThickness!(execute, self:Scriptable Tool Line Object Reference=@n4, Thickness:float=5) -> (then)
   > Add(
    TargetArray:Array of Scriptable Tool Line Object References=Get DragSquareLines()
    NewItem:"Scriptable Tool Line Object Reference (by ref)"=@n4
   )
   n5 = AddLine(execute=@n2.LoopBody, self=Get BboxLineSet() -> (BboxLineSet:Scriptable Tool Line Set Object Reference))
   > Add(TargetArray=Get BoxLines(), NewItem=@n5)
   Outputs()
  ]
 }
 Snapshot Actor Info {
  path=Setup.K2Node_Composite_4.Snapshot Actor Info
  nodes [
   Inputs()
   > n1 = GetAllLevelActors(self=Get EditorActorSubsystem())
   > n2 = Set AllActorsAtStart(AllActorsAtStart:Array of Actor Object References=@n1) -> (Output_Get:Array of Actor Object References)
   > n3 = For Each Loop(Array:Array of Actor Object References=@n2) -> (Array Element:Actor Object Reference)
   n4 = IsTemporarilyHiddenInEditor!(self:Actor Object Reference=@n3, bIncludeParent:bool=false) -> (ReturnValue:bool)
   > Add!(
    execute
    TargetMap:Map of Actor Object References to Booleans=Get IsHiddenAtStartOfTool()
    Key:"Actor Object Reference (by ref)"=@n3
    Value:"Boolean (by ref)"=@n4
   ) -> (then) {is=title K2Node_CallFunction}
   > Branch(Condition=@n4)
   > Add(TargetArray:Array of Actor Object References=Get ActorsToAlwaysIgnore_FromHidden(), NewItem:"Actor Object Reference (by ref)"=@n3) {is=title K2Node_CallArrayFunction}
   Outputs()
  ]
 }
 Settings Property Watches {
  path=Setup.K2Node_Composite_6.Settings Property Watches
  nodes [
   Inputs()
   > n1 = Sequence!(execute) -> (then_0, then_1, then_2:Exec, then_3:Exec)
   > WatchBoolProperty!(
    execute
    self:Scriptable Interactive Tool Object Reference
    PropertySet:Scriptable Interactive Tool Property Set Object Reference=Get Settings()
    PropertyName:String=Show Bounding Box Mesh
    OnModified:"Delegate (by ref)"=Create Event()
   ) -> (then, ReturnValue:Scriptable Interactive Tool Property Set Object Reference)
   n2 = Create Event()
   n3 = WatchProperty(execute=@n1.then_3, PropertySet=Get Settings(), PropertyName=Bounding Box Line Thickness, OnModified=@n2)
   > WatchProperty(PropertySet=@n3, PropertyName=Bounding Box Color, OnModified=@n2)
   WatchProperty(execute=@n1.then_1, PropertySet=Get Settings(), PropertyName=Object Visibility, OnModified=Create Event())
   n4 = For Each Loop(Exec=@n1.then_2, Array:Array of Strings=Get PropertiesAffectingFilters() -> (PropertiesAffectingFilters:Array of Strings)) -> (Array Element:String)
   > WatchProperty(PropertySet=Get Settings(), PropertyName=@n4, OnModified=Create Event())
   Outputs()
  ]
 }
 Other Settings Property Watches {
  path=Setup.K2Node_Composite_7.Other Settings Property Watches
  nodes [
   Inputs()
   n1 = Create Event()
   > WatchProperty(PropertySet=Get Settings(), PropertyName=Bounding Box Line Thickness, OnModified=@n1)
   > n2 = WatchProperty(PropertySet=Get Settings(), PropertyName=Bounding Box Color, OnModified=@n1)
   > n3 = WatchProperty(PropertySet=@n2, PropertyName=Filter Mode, OnModified=Create Event())
   n4 = Create Event()
   > n5 = WatchProperty(PropertySet=@n3, PropertyName=Exclude Label, OnModified=@n4)
   > n6 = WatchProperty(PropertySet=@n5, PropertyName=Exclude Class, OnModified=@n4)
   > n7 = WatchProperty(PropertySet=@n6, PropertyName=Include Label, OnModified=@n4)
   > WatchProperty(PropertySet=@n7, PropertyName=Include Class, OnModified=@n4)
   Outputs()
  ]
 }
 Init Sel Changed Detection {
  path=Setup.K2Node_Composite_9.Init Sel Changed Detection
  nodes [
   Inputs()
   > n1 = GetSelectedLevelActors(self=Get EditorActorSubsystem())
   > Set DetectSelChangedLastSet(DetectSelChangedLastSet=@n1)
   > Set SuspendSelChangeDetection(SuspendSelChangeDetection=false)
   Outputs()
  ]
 }
 ShouldRespondToMouseWheel {nodes [
   ShouldRespondToMouseWheel!() -> (then, CurrentPos:Input Device Ray Structure) {is=title K2Node_FunctionEntry}
   > Return Node!(
    execute
    ReturnValue_bHit:bool=Get IsModKeyPressed()
    ReturnValue_HitDepth:double=340282346638528859811704183484516925440
    ReturnValue_HitNormal:Vector="0,0,1"
    ReturnValue_bHasHitNormal:bool=False
    ReturnValue_HitIdentifier:int=0
    ReturnValue_HitObject:Object Reference
   )
  ]}
 ShowAll {nodes [
   ShowAll() {is=title K2Node_FunctionEntry}
   > Clear(TargetArray=Get HiddenActorsToIgnore())
   > Set HiddenActorsToIgnore(HiddenActorsToIgnore=Get ActorsToAlwaysIgnore_FromHidden())
   > n1 = "For Each Loop (Map)"!(execute:Exec, MapPin:Map of Actor Object References to Booleans=Get IsHiddenAtStartOfTool(), BreakPin:Exec) -> (then:Exec, KeyPin:Actor Object Reference, ValuePin:bool, CompletedPin:Exec)
   > SetIsTemporarilyHiddenInEditor(self=@n1, bIsHidden=@n1.ValuePin)
  ]}
 ShowOnlySelected {nodes [
   ShowOnlySelected() {is=title K2Node_FunctionEntry}
   > Clear(TargetArray=Get HiddenActorsToIgnore())
   > Set HiddenActorsToIgnore(HiddenActorsToIgnore=Get ActorsToAlwaysIgnore_FromHidden())
   > n1 = GetSelectedLevelActors(self=Get EditorActorSubsystem())
   > n2 = For Each Loop(Array:Array of Actor Object References=Get AllActorsAtStart()) -> (Array Element:Actor Object Reference)
   > Set CurrentActorInLoop(CurrentActorInLoop=@n2)
   n3 = Get Class Display Name(Class=Get Class(Object=Get CurrentActorInLoop()))
   > Branch(Condition="integer < integer"(A=FindSubstring(SearchIn=@n3, Substring=PreviewGeometryActor)))
   > n4 = Branch(Condition="integer >= integer"(A=Find Item(TargetArray=@n1, ItemToFind=@n2)))
   > SetIsTemporarilyHiddenInEditor(self=Get CurrentActorInLoop(), bIsHidden=false)
   SetIsTemporarilyHiddenInEditor(execute=@n4.else, self=Get CurrentActorInLoop(), bIsHidden=true)
   > Add(TargetArray=Get HiddenActorsToIgnore(), NewItem=Get CurrentActorInLoop())
  ]}
 SnapBoundsToSelection {nodes [
   SnapBoundsToSelection() {is=title K2Node_FunctionEntry}
   > UpdateSelectionBounds(ForceUpdate=true)
  ]}
 UpdateBBox {nodes [
   n1 = UpdateBBox!() -> (then, BoundOrigin:Vector=~, BoundExtent:Vector=~, Update Properties:bool=~) {is=title K2Node_FunctionEntry}
   > SetFadeCubeSMTransform(Location=@n1, Extent=@n1.BoundExtent) {is=title K2Node_CallFunction}
   > DrawBBox!(
    execute
    BoundOrigin:Vector=Get BoundOrigin() -> (BoundOrigin:Vector)
    BoundExtent:Vector=Get BoundExtent() -> (BoundExtent:Vector)
    Box Lines:Array of Scriptable Tool Line Object References=Get BoxLines()
    Color:Color Structure=Get Bounding Box Color!(self:BP STPS 3DSelect Object Reference=Get Settings()) -> (Bounding Box Color:Color Structure)
    Thickness:float=Get Bounding Box Line Thickness!(self:BP STPS 3DSelect Object Reference=Get Settings()) -> (Bounding Box Line Thickness:double)
    DepthBias:float=0
   ) -> (then)
   Get CurrentCorners()
  ]}
 UpdateBBoxAppearance {nodes [
   UpdateBBoxAppearance() {is=title K2Node_FunctionEntry}
   > Sequence!(execute) -> (then_0, then_1)
   > n1 = For Each Loop(Array:Array of Scriptable Tool Line Object References=Get BoxLines()) -> (Array Element:Scriptable Tool Line Object Reference)
   > SetLineColor!(execute, self=@n1, Color:Color Structure=Get Bounding Box Color(self=Get Settings())) -> (then)
   > SetLineThickness(self=@n1, Thickness=Get Bounding Box Line Thickness(self=Get Settings()))
  ]}
 UpdateFilterVariables {nodes [
   UpdateFilterVariables() {is=title K2Node_FunctionEntry}
   > n1 = Sequence!(execute) -> (then_0, then_1, then_2:Exec, then_3:Exec)
   > Set CurrentClassesToInclude(CurrentClassesToInclude:Array of Strings=Get Include Class(self=Get Settings())) -> (Output_Get:Array of Strings)
   Set CurrentLabelsToInclude(execute=@n1.then_1, CurrentLabelsToInclude:Array of Strings=Get Include Label(self=Get Settings())) -> (Output_Get:Array of Strings)
   Set CurrentLabelsToExclude(execute=@n1.then_2, CurrentLabelsToExclude:Array of Strings=Get Exclude Label(self=Get Settings())) -> (Output_Get:Array of Strings)
   Set CurrentClassesToExclude(execute=@n1.then_3, CurrentClassesToExclude:Array of Strings=Get Exclude Class(self=Get Settings())) -> (Output_Get:Array of Strings)
  ]}
 UpdateGizmoLocations {nodes [
   UpdateGizmoLocations() {is=title K2Node_FunctionEntry}
   > n1 = Branch(Condition=Get Use Edge Scaling(self=Get Settings()))
   > n2 = For Each Loop(Array:Array of Strings=Get AllGizmoIdentifiers()) -> (Array Element:String)
   > n3 = GetBoxFaceLocation(FaceIndex=@n2.Array Index)
   > SetGizmoTransform(Identifier=@n2, NewTransform="To Transform (Vector)"!(InLocation:Vector=@n3) -> (ReturnValue:Transform), bUndoable=false)
   SetGizmoTransform(
    execute=@n1.else
    Identifier=Get BoundGizmoIdentifier()
    NewTransform="To Transform (Vector)"(InLocation=Get SelectionBoundOrigin())
    bUndoable=false
   )
   Get Gizmo Identifier()
  ]}
 UpdateHoverText {nodes [
   UpdateHoverText() {is=title K2Node_FunctionEntry}
   n1 = BreakHitResult(Hit=Get HoverOutHit() -> (HoverOutHit:Hit Result Structure))
   > n2 = Branch(Condition=@n1)
   n3 = GetActorLabel(self=@n1.HitActor, bCreateIfNone=false)
   n4 = Len(S=@n3)
   > n5 = Branch(Condition="integer > integer"(A=@n4))
   > n6 = Set Class Line Value Class String(Class Line Value Class String:String=Get Class Display Name(Class=Get Class(Object=@n1.HitActor))) -> (Output_Get:String)
   n7 = Get Readout Line Actor()
   > "SetText (Text)"(self=Get LabelEntry!(self:BP UW Readout Line Object Reference=@n7) -> (LabelEntry:Common Text Object Reference), InText="LOCTEXT(\"Actor\")")
   > n8 = Set Actor Line Value(Actor Line Value:Text="To Text (String)"(InString=@n3)) -> (Output_Get:Text)
   n9 = Get Value(self=@n7)
   > "SetText (Text)"(self=@n9, InText=@n8)
   n10 = Get Readout Line Class()
   > "SetText (Text)"(self=Get LabelEntry(self=@n10), InText="LOCTEXT(\"Class\")")
   > n11 = Set ClassLineValue(ClassLineValue:Text="To Text (String)"(InString=@n6)) -> (Output_Get:Text)
   n12 = Get Value(self=@n10)
   > "SetText (Text)"(self=@n12, InText=@n11)
   "SetText (Text)"(execute=[@n2.else, @n5.else], self=@n9, InText="LOCTEXT(\"<none>\")")
   > "SetText (Text)"(self=@n12, InText="LOCTEXT(\"<none>\")")
   > Set Actor Line Value()
   > Set ClassLineValue()
   > Set Class Line Value Class String()
   "integer < integer"(A=@n4, B=1)
  ]}
 UpdateSelectedActorText {nodes [
   UpdateSelectedActorText() {is=title K2Node_FunctionEntry}
   n1 = Get Readout Line Selected() -> (Readout Line Selected:BP UW Readout Line Object Reference)
   > Is Valid(InputObject=@n1)
   > "SetText (Text)"(self=Get LabelEntry(self=@n1), InText=Get ActorCountLabel() -> (ActorCountLabel:Text))
   n2 = Get CurNumSelectedLevelActors() -> (CurNumSelectedLevelActors:int=0)
   > "SetText (Text)"(self=Get Value(self=@n1), InText="To Text (Integer)"(Value=@n2))
   > n3 = Branch(Condition="integer > integer"(A=@n2))
   n4 = vector * vector(A=Get SelectionBoundExtent(), B="2,2,2") -> (ReturnValue=~)
   n5 = BreakVector(InVec=@n4)
   n6 = "To Float (Integer)"(InInt=Truncate(A=@n5))
   n7 = "float > float"(A=GetMinElement!(A:Vector=@n4) -> (ReturnValue:double), B=1)
   n8 = "To Float (Integer)"(InInt=Truncate(A=@n5.Y))
   n9 = "To Float (Integer)"(InInt=Truncate(A=@n5.Z))
   n10 = Get Box Dimensions Line Format Value() -> (Box Dimensions Line Format Value:Text)
   > "SetText (Text)"(
    self=Get Value(self=Get Readout Line Dimensions() -> (Readout Line Dimensions:BP UW Readout Line Object Reference))
    InText=Format Text!(
     Format=@n10
     0:double=SelectFloat(A=@n6, B=@n5, bPickA=@n7)
     1:double=SelectFloat(A=@n8, B=@n5.Y, bPickA=@n7)
     2:double=SelectFloat(A=@n9, B=@n5.Z, bPickA=@n7)
    ) -> (Result)
   )
   n11 = BreakVector(InVec=@n4)
   n12 = "Max (Float)"!(A=@n11, B=@n11.Y) -> (ReturnValue)
   n13 = SafeDivide(A=Loge!(A:double="Clamp (Float)"(Value=@n12, Min=1, Max=1000000000)) -> (ReturnValue:double), B=Loge(A=2))
   n14 = "To Float (Integer)"(InInt=Ceil(A=@n13) {is=title K2Node_CallFunction})
   > "SetText (Text)"(
    self=Get Value(self=Get Readout Line Nearest Snap Grid() -> (Readout Line Nearest Snap Grid:BP UW Readout Line Object Reference))
    InText="To Text (Float)"(
     Value=Power!(Base:double=2, Exp:double=@n14) -> (ReturnValue:double) {is=title K2Node_CallFunction}
     MinimumIntegralDigits=1
     MaximumIntegralDigits=324
     MinimumFractionalDigits=0
     MaximumFractionalDigits=3
    )
   )
   n15 = Make Literal Float(Value=0)
   "SetText (Text)"(execute=@n3.else, self=Get Value(self=Get Readout Line Dimensions()), InText=Format Text(Format=@n10, 0=@n15, 1=@n15, 2=@n15))
   > "SetText (Text)"(self=Get Value(self=Get Readout Line Nearest Snap Grid()), InText="LOCTEXT(\"0\")")
   Floor(A=@n13) {is=title K2Node_CallFunction}
  ]}
 UpdateSelectedActors {nodes [
   n1 = UpdateSelectedActors!() -> (then, BoxPos:Vector=~, BoxExtent:Vector=~) {is=title K2Node_FunctionEntry}
   > Clear(TargetArray=Get FinalOverlapActors() -> (FinalOverlapActors:Array of Actor Object References))
   n2 = GetBoxCenterSize!(Box:Box Structure=GetExpandedBox!(Box:Box Structure=MakeBoxFromCenterExtents(Center=@n1, Extents=@n1.BoxExtent), ExpandBy:Vector="5,5,5") -> (ReturnValue:Box Structure)) -> (Center:Vector, Dimensions:Vector)
   > n3 = Box Overlap Actors!(
    execute
    BoxPos:Vector=@n2
    BoxExtent:Vector=vector * vector(A=@n2.Dimensions, B="0.5,0.5,0.5") -> (ReturnValue=~)
    ObjectTypes:Array of EObjectTypeQuery Enums=Make Array!(
     "[0]":EObjectTypeQuery Enum=ObjectTypeQuery1
     "[1]":EObjectTypeQuery Enum=ObjectTypeQuery2
     "[2]":EObjectTypeQuery Enum=ObjectTypeQuery6
     "[3]":EObjectTypeQuery Enum=ObjectTypeQuery4
     "[4]":EObjectTypeQuery Enum=ObjectTypeQuery5
    ) -> (Array:Array of EObjectTypeQuery Enums)
    ActorClassFilter:Object Class Reference=&Actor
    ActorsToIgnore:Array of Actor Object References=Get ActorsToAlwaysIgnore_FromHidden()
   ) -> (then, OutActors:Array of Actor Object References, ReturnValue:bool)
   > n4 = For Each Loop(Array:Array of Actor Object References=@n3) -> (Array Element:Actor Object Reference)
   Remove Item(execute=@n4.Completed, TargetArray=Get FinalOverlapActors(), Item=Get FadeCubeActor())
   > Clear(TargetArray:Array of Booleans=Get OverlapActorsIsTempHidden() -> (OverlapActorsIsTempHidden:Array of Booleans))
   > n5 = For Each Loop(Array=Get FinalOverlapActors())
   SetSelectedLevelActors_3DSel(execute=@n5.Completed, ActorsToSelect=Get FinalOverlapActors()) {is=title K2Node_CallFunction}
   > n6 = For Each Loop(Array=Get FinalOverlapActors())
   UpdateSelectedActorText(execute=@n6.Completed) {is=title K2Node_CallFunction}
   SetIsTemporarilyHiddenInEditor(
    execute=@n6.LoopBody
    self=@n6
    bIsHidden="Get (a ref)"(Array:Array of Booleans=Get OverlapActorsIsTempHidden(), Dimension 1=@n6.Array Index) -> (Output:"Boolean (by ref)")
   )
   Add(
    execute=@n5.LoopBody
    TargetArray:Array of Booleans=Get OverlapActorsIsTempHidden()
    NewItem:"Boolean (by ref)"=IsTemporarilyHiddenInEditor(self=@n5, bIncludeParent=false)
   )
   > SetIsTemporarilyHiddenInEditor(self=@n5, bIsHidden=false)
   Set CurActor(execute=@n4.LoopBody, CurActor:Actor Object Reference=@n4) -> (Output_Get:Actor Object Reference)
   n7 = Get Class Display Name(Class=Get Class(Object=Get CurActor() -> (CurActor:Actor Object Reference)))
   > n8 = Branch(Condition="integer >= integer"(A=FindSubstring(SearchIn=@n7, Substring=gizmo)))
   n9 = Get Actor Bounds(self=Get CurActor(), bOnlyCollidingComponents=false)
   n10 = MakeBoxFromCenterExtents(Center=@n9, Extents=@n9.BoxExtent)
   n11 = MakeBoxFromCenterExtents(Center=Get BoxPos() -> (BoxPos:Vector), Extents=Get BoxExtent() -> (BoxExtent:Vector))
   n12 = GetBoxVolumeArea!(Box:Box Structure=FindBoxBoxIntersection!(Box1:Box Structure=@n10, Box2:Box Structure=@n11) -> (bIsIntersecting:bool, ReturnValue:Box Structure).ReturnValue) -> (Volume:double, SurfaceArea:double)
   n13 = GetBoxVolumeArea(Box=@n10)
   Branch(
    execute=@n8.else
    Condition="integer64 >= integer64"(A:Integer64=Truncate to Integer64!(A:double=@n12) -> (ReturnValue:Integer64=0), B:Integer64=Truncate to Integer64(A=@n13) -> (ReturnValue=0)) -> (ReturnValue:bool=~) {is=title K2Node_PromotableOperator}
   )
   > Add(TargetArray:Array of Actor Object References=Get FinalOverlapActors(), NewItem:"Actor Object Reference (by ref)"=Get CurActor())
   "Is Inside Or On (Box)"!(InnerTest:"Box Structure (by ref)"=@n10, OuterTest:"Box Structure (by ref)"=@n11) -> (ReturnValue:bool)
  ]}
 UpdateSelectionBounds {nodes [
   n1 = UpdateSelectionBounds!() -> (then, ForceUpdate:bool=~) {is=title K2Node_FunctionEntry}
   > Branch(Condition=@n1)
   > GetSelectionBounds() {is=title K2Node_CallFunction}
   > UpdateSelectedActorText()
   > UpdateBBox(BoundOrigin=Get SelectionBoundOrigin(), BoundExtent=Get SelectionBoundExtent(), Update Properties=false) {is=title K2Node_CallFunction}
  ]}
 WaitForTime {nodes [
   n1 = WaitForTime!() -> (then, PauseTime:double=0.5) {is=title K2Node_FunctionEntry}
   n2 = Get LastWaitTime() -> (LastWaitTime:double)
   > n3 = Branch(Condition="float > float"(A=Get CurTime(), B=float + float(A=@n1, B=@n2)))
   > Set LastWaitTime(LastWaitTime:double=Get CurTime()) -> (Output_Get:double)
   > Return Node!(execute, Active:bool=true)
   Return Node(execute=@n3.else, Active=false)
  ]}
]
variables [
 Actor Line Label {type=string, default=Actor}
 Actor Line Value {type=string, default="<none>"}
 ActorCountLabel {type=string, default=Selected Objects}
 ActorsToAlwaysIgnore_FromHidden {type="array<Actor>"}
 AllActorsAtStart {type="array<Actor>"}
 AllGizmoIdentifiers {type="array<string>", default []}
 BboxLineSet {type=ScriptableToolLineSet, default=null}
 BoundEdgeGizmoIdentifier {type=string, default=BoundEdge_}
 BoundGizmoIdentifier {type=string, default=Bounds}
 Box Dimensions Label {type=string, default=Dimensions}
 Box Dimensions Line Empty Value {type=string, default=0 x 0 x 0}
 Box Dimensions Line Format Value {type=string, default="{0} x {1} x {2}"}
 BoxLines {type="array<ScriptableToolLine>", default []}
 Class Line Value Class String {type=string, default=None}
 ClassLineLabel {type=string, default=Class}
 ClassLineValue {type=string, default="<none>"}
 Clear Warning After Time {type=number, default=5}
 ClickDragHitBeforeDrag {
  type=HitResult
  default {
   bBlockingHit=false
   bStartPenetrating=false
   boneName=None
   component=null
   distance=0
   elementIndex=0
   faceIndex=0
   item=0
   myBoneName=None
   myItem=-1
   penetrationDepth=0
   physMaterial=null
   time=1
   hitObjectHandle {referenceObject=null}
   impactNormal=<0 0 0>
   impactPoint=<0 0 0>
   location=<0 0 0>
   normal=<0 0 0>
   traceEnd=<0 0 0>
   traceStart=<0 0 0>
  }
 }
 ClickDragWorldRayBeforeDrag {type=Ray, default {direction=<0 0 0>, origin=<0 0 0>}}
 Collision Shown {type=boolean}
 CurNumSelectedLevelActors {type=integer}
 CurrentClassesToExclude {
  type="array<string>"
  default [&Class'/Script/Landscape.LandscapeMeshProxyActor', &Class'/Script/Landscape.LandscapeStreamingProxy']
 }
 CurrentClassesToInclude {type="array<string>", default []}
 CurrentCorners {type="array<Vector>"}
 CurrentLabelsToExclude {type="array<string>", default [landscape]}
 CurrentLabelsToInclude {type="array<string>", default []}
 CurSelectedLevelActors {type="array<Actor>"}
 CurSelectionSets {type="array<string>", default []}
 CurTime {type=number}
 CycleFilterKey {type=Key, default=C}
 CycleObjVisibilityKey {type=Key, default=V}
 DetectSelChangedLastSet {type="array<Actor>"}
 DoneFirstInit {type=boolean}
 DragSquareLines {type="array<ScriptableToolLine>", default []}
 DragSquareLineSet {type=ScriptableToolLineSet, default=null}
 ExcludeClassKey {type=Key, default=N}
 ExcludeStringKey {type=Key, default=N}
 ExtentBeforeGizmo {type=Vector}
 FadeCubeActor {type=StaticMeshActor, default=null}
 FadeCubeActorStaticMeshComponent {type=StaticMeshComponent, default=null}
 FirstInitRun {type=boolean}
 Header Widget {type=BP_UW_SharedHeaderWidget_C}
 HiddenActorsToIgnore {type="array<Actor>"}
 HoverOutHit {
  type=HitResult
  default {
   bBlockingHit=false
   bStartPenetrating=false
   boneName=None
   component=null
   distance=0
   elementIndex=0
   faceIndex=0
   item=0
   myBoneName=None
   myItem=-1
   penetrationDepth=0
   physMaterial=null
   time=1
   hitObjectHandle {referenceObject=null}
   impactNormal=<0 0 0>
   impactPoint=<0 0 0>
   location=<0 0 0>
   normal=<0 0 0>
   traceEnd=<0 0 0>
   traceStart=<0 0 0>
  }
 }
 IncludeClassKey {type=Key, default=B}
 IncludeStringKey {type=Key, default=B}
 IsHiddenAtStartOfTool {}
 IsModKeyPressed {type=boolean}
 Last Clear Time {type=number}
 LastWaitTime {type=number}
 LegendWidget {type=BP_UW_3DSelectHeader_C, default=null}
 ModKey {type=Key, default=X}
 Nearest Grid Label {type=string, default=Best Fit Snap Grid}
 Nearest Grid Value {type=string, default=0}
 Only Colliding Components {type=boolean, default=true}
 OriginBeforeGizmo {type=Vector}
 PropSetSaveKey {type=string, default=SelectionSizer}
 PropSetSettingsIdentifier {type=string, default=Settings}
 Readout Box {type=BP_UW_ExpandableBox_C}
 Readout Line Actor {type=BP_UW_ReadoutLine_C, default=null}
 Readout Line Class {type=BP_UW_ReadoutLine_C, default=null}
 Readout Line Dimensions {type=BP_UW_ReadoutLine_C, default=null}
 Readout Line Nearest Snap Grid {type=BP_UW_ReadoutLine_C, default=null}
 Readout Line Selected {type=BP_UW_ReadoutLine_C, default=null}
 Selected Line Label {type=string, default=Selected Objects}
 Selected Line Value {type=string, default=0}
 SelectionBoundExtent {type=Vector}
 SelectionBoundOrigin {type=Vector}
 SelectionOfActorRemovedByFilter_WarningMessage {type=string, default="Warning: Selection of {0} prevented by Filters"}
 Settings {type=BP_STPS_3DSelect_C, default=null}
 SharedState Settings {type=BP_STPS_SharedWidgetState_C, default=null}
 SharedStateIdentifier {type=string, default=SharedStateSettings}
 Shortcut Box {type=BP_UW_ExpandableBox_C}
 ShortcutKeys {
  type="array<string>"
  default [
   Resize Volume
   Scale Volume Uniformly
   Cycle Object Visibility
   Cycle Filter Mode
   Include Class
   Include Label
   Exclude Class
   Exclude Label
  ]
 }
 ShortcutLabels {
  type="map<string>"
  default {
   Cycle Filter Mode=Press C
   Cycle Object Visibility=Press V
   Exclude Class=Press N
   Exclude Label=Press Shift + N
   Include Class=Press B
   Include Label=Press Shift + B
   Resize Volume=Hold X
   Scale Volume Uniformly=Hold X + Mouse Wheel
  }
 }
 ShortcutLines {type="array<BP_UW_ReadoutLine_C>", default []}
 ShowHideCollisionKey {type=Key, default=C}
 SingleClickHit {
  type=HitResult
  default {
   bBlockingHit=false
   bStartPenetrating=false
   boneName=None
   component=null
   distance=0
   elementIndex=0
   faceIndex=0
   item=0
   myBoneName=None
   myItem=-1
   penetrationDepth=0
   physMaterial=null
   time=1
   hitObjectHandle {referenceObject=null}
   impactNormal=<0 0 0>
   impactPoint=<0 0 0>
   location=<0 0 0>
   normal=<0 0 0>
   traceEnd=<0 0 0>
   traceStart=<0 0 0>
  }
 }
 StartupLabelsToExclude {type="array<string>", default [landscape]}
 SuspendSelChangeDetection {type=boolean}
 Undo Context {type=string, default=ThreeD Select}
]
